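# Sound Cue that plays the "How To Play Music Ascend" music track, stored with the SparksMaps music audio.
# A single looping Wave Player is the root, with no randomization, mixing or other processing nodes.
# The cue lowers its volume multiplier and routes through the Fort_SFX_InGame sound class, an in-game SFX class by name.
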SoundCue /SparksMaps/Audio/Music/How_To_Play_Music_Ascend_Cue
root=WavePlayer_0
nodes [{id=WavePlayer_0, properties {bLooping=true, soundWaveAssetPtr=&/How_To_Play_Music_Ascend}}]
properties {maxDistance=2097152, soundClassObject=&Fort_SFX_InGame, volumeMultiplier=0.3}
