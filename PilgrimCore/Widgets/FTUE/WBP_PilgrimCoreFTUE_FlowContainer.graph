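# Activatable FTUE container for Pilgrim that shows a sequence of onboarding screens with music, page pips and Back and skip buttons.
# On construct it fills a switcher from WidgetsToShow, fades in music and listens for FTUE Advance and Complete router events to page forward or close.
# Page changes update pips and button visibility, then focus the active screen's desired target a tick later.
# Closing stops both listeners and fades out music.
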WidgetBlueprint /PilgrimCore/Widgets/FTUE/WBP_PilgrimCoreFTUE_FlowContainer
graphs [
 BP_GetDesiredFocusTarget {nodes [
   Get Desired Focus Target()
   > n1 = GetWidgetToFocus!(execute, self) -> (then, WidgetToFocus:Widget Object Reference)
   > Return Node!(execute, ReturnValue:Widget Object Reference=@n1)
  ]}
 EventGraph {nodes [
   n1 = "On Clicked (Back_Button)"!() -> (OutputDelegate, then, Button:Common Button Base Object Reference) {is=title K2Node_ComponentBoundEvent}
   > Fire Event Element Interacted(Element=@n1.Button)
   n2 = Get FTUEContentSwitcher() -> (FTUEContentSwitcher:Common Activatable Widget Switcher Object Reference)
   > n3 = Branch(Condition="Not Equal (Integer)"(A=GetActiveWidgetIndex!(self:Widget Switcher Object Reference=@n2) -> (ReturnValue:int=0)))
   > ActivatePreviousWidget!(execute, self:Common Animated Switcher Object Reference=Get FTUEContentSwitcher(), bCanWrap:bool=false) -> (then)
   CleanupWidget(execute=@n3.else)
   > Broadcast Event!(
    execute
    self:Gameplay Event Router Component Object Reference=GetGlobalEventRouter()
    EventData:FMEvent Calibration Toggle Modal Structure=Make FMEvent Calibration Toggle Modal!() -> (FMEvent_Calibration_ToggleModal:FMEvent Calibration Toggle Modal Structure) {is=title K2Node_MakeStruct}
    EventContext:Object Reference
   ) -> (then, ReturnValue:bool)
   n4 = "On Active Widget Index Changed BP (FTUEContentSwitcher)"!() -> (OutputDelegate, then, ActiveWidget:Widget Object Reference, ActiveIndex:int=0) {is=title K2Node_ComponentBoundEvent}
   > SetCurrentIndex!(
    execute
    self:WBP Pilgrim Parent Pip List Object Reference=Get WBP_Pilgrim_Parent_Pip_List() -> (WBP_Pilgrim_Parent_Pip_List:WBP Pilgrim Parent Pip List Object Reference)
    RawIndex:int=@n4.ActiveIndex
   ) -> (then)
   > SetVisibility(
    self=Get Back_Button() -> (Back_Button:WBP UIKit Button Quiet Object Reference)
    InVisibility=Select(Option 0=SelfHitTestInvisible, Option 1=Hidden, Index="Equal (Integer)"(A=@n4.ActiveIndex))
   )
   n5 = int - int(A=GetChildrenCount!(self:Panel Widget Object Reference=Get FTUEContentSwitcher()) -> (ReturnValue:int=0))
   > SetVisibility(
    self=Get WBP_FMCalibration_ConfirmSkipButton() -> (WBP_FMCalibration_ConfirmSkipButton:WBP FMCalibration Skip Button Object Reference)
    InVisibility=Select(Option 0=SelfHitTestInvisible, Option 1=Collapsed, Index="Equal (Integer)"(A=@n5, B=@n4.ActiveIndex))
   )
   > SetFocusForFTUEScreen()
   "On Popup Closed (WBP_FMCalibration_ConfirmSkipButton)"() {is=title K2Node_ComponentBoundEvent}
   > SetFocusForFTUEScreen()
   CleanupWidget() {is=title K2Node_CustomEvent}
   > StopListeningToEvent(self:Async Action Start Listening To Event Object Reference=[
     Get FTUECompleteAsyncTask() -> (FTUECompleteAsyncTask:Async Action Start Listening To Event Object Reference)
     Get FTUEAdvance Async Task() -> (FTUEAdvance Async Task:Async Action Start Listening To Event Object Reference)
    ])
   > DeactivateWidget()
   n6 = Get MusicComponent() -> (MusicComponent:Audio Component Object Reference)
   > Is Valid(InputObject=@n6)
   > FadeOut!(execute, self:Audio Component Object Reference=@n6, FadeOutDuration:float=3, FadeVolumeLevel:float=0, FadeCurve:EAudioFaderCurve Enum=Linear) -> (then)
   SetFocusForFTUEScreen() {is=title K2Node_CustomEvent}
   > n7 = GetWidgetToFocus()
   > DelayUntilNextTick!(execute) -> (then)
   > SetFocus(self=@n7)
   > Fire Event Main View Opened!(execute, View:Widget Object Reference=@n7, AdditionalInfo:Flow Analytics Additional Info Structure, bIsOverlay:bool=false) -> (then)
   Event On Activated()
   > Fire Event Main View Opened(View=Self-Reference(), bIsOverlay=false)
   > SetFocusForFTUEScreen() {is=title K2Node_CallFunction}
   Event On Deactivated()
   > StopListeningToEvent(self=[Get FTUEAdvance Async Task(), Get FTUECompleteAsyncTask()])
   n8 = Get MusicComponent()
   > Is Valid(InputObject=@n8)
   > FadeOut(self=@n8, FadeOutDuration=2, FadeVolumeLevel=0, FadeCurve=Linear)
   Event Construct()
   > SetupPlayerIdentity!(
    execute
    self:WBP Pilgrim Player Identity Pill Object Reference=Get WBP_Pilgrim_PlayerIdentity_Pill() -> (WBP_Pilgrim_PlayerIdentity_Pill:WBP Pilgrim Player Identity Pill Object Reference)
    OptionalPlayerState:Player State Object Reference
    ForceVisible:bool=false
   ) -> (then)
   > SetAllNavigationRules!(execute, self:Widget Object Reference, Rule:EUINavigationRule Enum=Stop, WidgetToFocus:Name=None) -> (then)
   > n9 = Branch(Condition=Is Not Empty!(TargetArray:Array of Widget Class References=Get WidgetsToShow() -> (WidgetsToShow:Array of Widget Class References)) -> (ReturnValue:bool))
   > n10 = Sequence!(execute) -> (then_0, then_1, then_2:Exec)
   n11 = Get MusicCue() -> (MusicCue:Sound Base Object Reference)
   > Is Valid(InputObject=@n11)
   > n12 = SpawnSound2D!(
    execute
    Sound:Sound Base Object Reference=@n11
    VolumeMultiplier:float=1
    PitchMultiplier:float=1
    StartTime:float=0
    ConcurrencySettings:Sound Concurrency Object Reference
    bPersistAcrossLevelTransition:bool=false
    bAutoDestroy:bool=true
   ) -> (then, ReturnValue:Audio Component Object Reference)
   > n13 = Set MusicComponent(MusicComponent:Audio Component Object Reference=@n12) -> (Output_Get:Audio Component Object Reference)
   > FadeIn!(
    execute
    self:Audio Component Object Reference=@n13
    FadeInDuration:float=2
    FadeVolumeLevel:float=1
    StartTime:float=0
    FadeCurve:EAudioFaderCurve Enum=Linear
   ) -> (then)
   n14 = For Each Loop(Exec=@n10.then_1, Array:Array of Widget Class References=Get WidgetsToShow()) -> (Array Element:Widget Class Reference)
   SetFocusForFTUEScreen(execute=@n14.Completed)
   > FillEntryBox(self:WBP Pilgrim Parent Pip List Object Reference=Get WBP_Pilgrim_Parent_Pip_List())
   n15 = Construct Object from Class(execute=@n14.LoopBody, Class=@n14) -> (ReturnValue:Widget Object Reference)
   > AddChild(self=Get FTUEContentSwitcher(), Content=@n15)
   > Add(
    TargetArray:Array of Integers=Get IndexMap!(self:WBP Pilgrim Parent Pip List Object Reference=Get WBP_Pilgrim_Parent_Pip_List()) -> (IndexMap:Array of Integers)
    NewItem:"Integer (by ref)"=1
   ) {is=title K2Node_CallArrayFunction}
   n16 = GetEventRouter!(execute=@n10.then_2, Scope:Object Class Reference=/Script/FortniteGame.FortPlayspace, ContextActor:Actor Object Reference=GetOwningPlayer()) -> (then, ReturnValue:Gameplay Event Router Component Object Reference)
   > Is Valid(InputObject=@n16)
   > n17 = Set EventRouter(EventRouter:Gameplay Event Router Component Object Reference=@n16) -> (Output_Get:Gameplay Event Router Component Object Reference)
   > n18 = Sequence!(execute) -> (then_0, then_1)
   > n19 = StartListeningToEvent!(
    execute:Exec
    Target:Gameplay Event Router Component Object Reference=@n17
    EventType:Script Struct Object Reference=/Script/PilgrimCoreUI.PilgrimEvent_FTUE_Complete
    Context:Object Reference
    EventBubblingRule:EEventBubblingRule Enum
   ) -> (
    then:Exec
    AsyncTaskProxy:Async Action Start Listening To Event Object Reference
    OnEventReceived:Exec
    Payload:Pilgrim Event FTUE Complete Structure
   )
   > Set FTUECompleteAsyncTask(FTUECompleteAsyncTask:Async Action Start Listening To Event Object Reference=@n19) -> (Output_Get:Async Action Start Listening To Event Object Reference)
   CleanupWidget(execute=@n19.OnEventReceived) {is=title K2Node_CallFunction}
   n20 = StartListeningToEvent(execute=@n18.then_1, Target=@n17, EventType=/Script/PilgrimCoreUI.PilgrimEvent_FTUE_Advance) -> (Payload:Pilgrim Event FTUE Advance Structure)
   > Set FTUEAdvance Async Task(FTUEAdvance Async Task:Async Action Start Listening To Event Object Reference=@n20) -> (Output_Get:Async Action Start Listening To Event Object Reference)
   ActivateNextWidget!(execute=@n20.OnEventReceived, self:Common Animated Switcher Object Reference=Get FTUEContentSwitcher(), bCanWrap:bool=false) -> (then)
   DeactivateWidget(execute=@n9.else)
   > PrintString!(
    execute
    InString=No widgets set to show for the FTUE sequence. Deactivating the FTUE screen.
    bPrintToScreen
    bPrintToLog
    TextColor
    Duration
    Key
   ) -> (then)
  ]}
 GetWidgetToFocus {nodes [
   GetWidgetToFocus() {is=title K2Node_FunctionEntry}
   > n1 = Cast To CommonActivatableWidget!(execute, Object=GetActiveWidget!(self:Widget Switcher Object Reference=Get FTUEContentSwitcher()) -> (ReturnValue:Widget Object Reference)) -> (then, CastFailed, AsCommon Activatable Widget:Common Activatable Widget Object Reference) {is=title K2Node_DynamicCast}
   > Return Node!(
    execute
    WidgetToFocus:Widget Object Reference=GetDesiredFocusTarget!(self:Common Activatable Widget Object Reference=@n1) -> (ReturnValue:Widget Object Reference)
   )
   Return Node(execute=@n1.CastFailed, WidgetToFocus=Get FTUEContentSwitcher())
  ]}
]
variables [
 EventRouter {type=GameplayEventRouterComponent, default=null}
 FTUEAdvance Async Task {type=AsyncAction_StartListeningToEvent, default=null}
 FTUECompleteAsyncTask {type=AsyncAction_StartListeningToEvent, default=null}
 MusicComponent {type=AudioComponent, default=null}
 MusicCue {type=SoundBase, default=/SparksMaps/Audio/Music/How_To_Play_Music_Ascend_Cue}
 WidgetsToShow {type="array<Class>", default []}
]
widgets {
 info {inheritedWidgetCount=0, parentClass=&CommonActivatableWidget, rootWidgetClass=&Overlay, widgetCount=7}
 widgets [
  Overlay_20 {widgetClassPath=&Overlay, widget=Overlay_20}
  FTUEContentSwitcher {
   widgetClassPath=&CommonActivatableWidgetSwitcher
   bIsVariable=true
   parent=Overlay_20
   widget=FTUEContentSwitcher
   properties {transitionCurveType=Linear, transitionDuration=0, transitionType=Horizontal}
   slot {class=OverlaySlot, properties {horizontalAlignment=HAlign_Fill, verticalAlignment=VAlign_Fill}}
  }
  WBP_Pilgrim_Parent_Pip_List {
   widgetClassPath=/SparksUICommon/Widgets/WBP_Pilgrim_Parent_Pip_List
   bIsVariable=true
   parent=Overlay_20
   widget=WBP_Pilgrim_Parent_Pip_List
   properties {indexMap []}
   slot {class=OverlaySlot, properties {horizontalAlignment=HAlign_Center, padding=<ltrb 0 30 0 0>}}
  }
  BottomButtons {
   widgetClassPath=&GridPanel
   parent=Overlay_20
   widget=BottomButtons
   slot {class=OverlaySlot, properties {horizontalAlignment=HAlign_Right, verticalAlignment=VAlign_Bottom}}
  }
  WBP_FMCalibration_ConfirmSkipButton {
   widgetClassPath=/FMCalibration/Widgets/WBP_FMCalibration_SkipButton
   bIsVariable=true
   parent=BottomButtons
   widget=WBP_FMCalibration_ConfirmSkipButton
   properties {bIsFocusable=false}
   slot {class=GridSlot, properties {column=1, padding=<ltrb 20 20 20 20>}}
  }
  Back_Button {
   widgetClassPath=/FortUILibrary/UIKit/Buttons/WBP_UIKit_Button_Quiet
   bIsVariable=true
   parent=BottomButtons
   widget=Back_Button
   properties {bIsFocusable=false, text=Back, visibility=Hidden, triggeringInputAction {dataTable=&InputActionDataTable, rowName=Back}}
   slot {class=GridSlot, properties {padding=<ltrb 20 20 20 20>}}
  }
  WBP_Pilgrim_PlayerIdentity_Pill {
   widgetClassPath=/PilgrimCore/Widgets/Player/WBP_Pilgrim_PlayerIdentity_Pill
   bIsVariable=true
   parent=Overlay_20
   widget=WBP_Pilgrim_PlayerIdentity_Pill
   properties {padding=<ltrb 0 23 23 0>}
   slot {class=OverlaySlot, properties {horizontalAlignment=HAlign_Right}}
  }
 ]
}
