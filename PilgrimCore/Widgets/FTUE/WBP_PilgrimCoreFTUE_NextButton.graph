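# WBP_PilgrimCoreFTUE_NextButton is the Next button widget for the Pilgrim first-time user experience, advancing it when clicked.
# It subclasses the UIKit Loud button and adds no widgets of its own.
# On click, it gets the playspace-scoped gameplay event router from the owning player and, if valid, broadcasts an FTUE Advance event, leaving the response to whatever listens.
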WidgetBlueprint /PilgrimCore/Widgets/FTUE/WBP_PilgrimCoreFTUE_NextButton
graphs [EventGraph {nodes [
   Event On Clicked()
   > n1 = GetEventRouter!(execute, Scope:Object Class Reference=/Script/FortniteGame.FortPlayspace, ContextActor:Actor Object Reference=GetOwningPlayer()) -> (then, ReturnValue:Gameplay Event Router Component Object Reference)
   > Is Valid(InputObject=@n1)
   > Broadcast Event!(
    execute
    self:Gameplay Event Router Component Object Reference=@n1
    EventData:Pilgrim Event FTUE Advance Structure=Make Pilgrim Event FTUE Advance!() -> (PilgrimEvent_FTUE_Advance:Pilgrim Event FTUE Advance Structure) {is=title K2Node_MakeStruct}
    EventContext:Object Reference
   ) -> (then, ReturnValue:bool)
  ]}]
variables []
widgets {
 info {
  inheritedWidgetCount=0
  parentClass=/FortUILibrary/UIKit/Buttons/WBP_UIKit_Button_Loud.WBP_UIKit_Button_Loud_C
  rootWidgetClass=null
  widgetCount=0
 }
 widgets []
}
