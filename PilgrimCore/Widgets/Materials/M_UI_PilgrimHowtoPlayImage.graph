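# Translucent UI-domain material for the Pilgrim how-to-play image, judging by its name and default texture.
# A texture parameter named Texture, defaulting to T_UI_HowtoplayImage, samples the image with standard texture coordinates.
# MF_UI_CoreUI_RoundedMask builds a rounded-corner mask from constant size and uniform corner radius inputs.
# Only the texture is exposed as a parameter, so the corner shape is fixed in the material.
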Material /PilgrimCore/Widgets/Materials/M_UI_PilgrimHowtoPlayImage
graph {
 properties {
  bAutomaticallySetUsageInEditor=false
  bUsedWithStaticMesh=false
  blendMode=BLEND_Translucent
  materialDomain=MD_UI
  pixelDepthOffsetMode=PDOM_Legacy
  usedShadingModels=Default Lit
 }
 nodes [
  MaterialFunctionCall(Size=Constant2Vector() {g=1, r=1}, Corner Radius=Constant4Vector() {constant=<rgba 0.1 0.1 0.1 0.1>}.RGBA) {materialFunction=&MF_UI_CoreUI_RoundedMask}
  TextureSampleParameter2D(UVs=TextureCoordinate()) {parameterName=Texture, texture=/PilgrimCore/Widgets/Textures/T_UI_HowtoplayImage}
 ]
}
