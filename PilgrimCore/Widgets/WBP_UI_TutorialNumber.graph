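# WBP_UI_TutorialNumber is a small fixed-size badge widget that displays a number, apparently for tutorial UI.
# A square SizeBox wraps a yellow-tinted, material-backed Border, which centers a bold black CommonTextBlock named NUMER.
# PreConstruct copies the string variable Number into that text block, replacing its placeholder text at design time and at runtime.
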WidgetBlueprint /PilgrimCore/Widgets/WBP_UI_TutorialNumber
graphs [EventGraph {nodes [
   Event PreConstruct()
   > Set Text(Text=Get Number() -> (Number:Text), self=Get NUMER() -> (NUMER:Common Text Object Reference))
  ]}]
variables [Number {type=string, default=0}]
widgets {
 info {inheritedWidgetCount=0, parentClass=&UserWidget, rootWidgetClass=&SizeBox, widgetCount=3}
 widgets [
  SizeBox_0 {
   widgetClassPath=&SizeBox
   widget=SizeBox_0
   properties {bOverride_HeightOverride=true, bOverride_WidthOverride=true, heightOverride=40, widthOverride=40}
  }
  BG {
   widgetClassPath=&Border
   parent=SizeBox_0
   widget=BG
   properties {
    horizontalAlignment=HAlign_Center
    verticalAlignment=VAlign_Center
    background {
     bIsDynamicallyLoaded=false
     drawAs=Image
     imageType=FullColor
     mirroring=NoMirror
     resourceName=None
     resourceObject=/PilgrimCore/Widgets/Materials/MI_UI_Pilgrim_TutorialNumberBG
     tiling=NoTile
     imageSize=<1 1>
     margin=<ltrb 0 0 0 0>
     outlineSettings {
      bUseBrushTransparency=false
      roundingType=HalfHeightRadius
      width=0
      color {colorUseRule=UseColor_Specified, specifiedColor=<rgba 0 0 0 0>}
      cornerRadii=<0 0 0 0>
     }
     tintColor {colorUseRule=UseColor_Specified, specifiedColor=<rgba 1 0.907 0 1>}
     uVRegion {bIsValid=false, max=<0 0>, min=<0 0>}
    }
    padding=<ltrb 0 0 0 0>
   }
   slot {class=SizeBoxSlot}
  }
  NUMER {
   widgetClassPath=&CommonTextBlock
   bIsVariable=true
   parent=BG
   widget=NUMER
   properties {
    text=1
    colorAndOpacity {colorUseRule=UseColor_Specified, specifiedColor=<rgba 0 0 0 1>}
    font {
     bForceMonospaced=false
     bMaterialIsStencil=false
     fontMaterial=null
     fontObject=&HeadingNow
     letterSpacing=0
     monospacedWidth=1
     size=19.5
     skewAmount=0
     typefaceFontName=86 Bold
     outlineSettings {
      bApplyOutlineToDropShadows=false
      bMiteredCorners=false
      bSeparateFillAlpha=false
      outlineMaterial=null
      outlineSize=0
      outlineColor=<rgba 0 0 0 1>
     }
    }
    shadowOffset=<0 0>
   }
   slot {class=BorderSlot, properties {horizontalAlignment=HAlign_Center, verticalAlignment=VAlign_Center, padding=<ltrb 0 0 0 0>}}
  }
 ]
}
