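# First-time-user "How To Play" screen for Pilgrim's music game, explaining note hitting, streak multipliers and star scoring.
# A tutorial image with three overlaid number markers sits beside three numbered tips, above a Next button on a full-screen material background.
# Built on PilgrimFTUEWidgetBase with an empty event graph; its only logic makes the Next button the desired focus target.
# It also layers in a WBP_CaptureForPostBufferUpdate widget behind the content panel.
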WidgetBlueprint /PilgrimCore/Widgets/FTUE/WBP_PilgrimCoreFTUE_HowToPlay
graphs [
 BP_GetDesiredFocusTarget {nodes [
   Get Desired Focus Target()
   > Return Node!(
    execute
    ReturnValue:Widget Object Reference=Get WBP_PilgrimCoreFTUE_NextButton() -> (WBP_PilgrimCoreFTUE_NextButton:WBP Pilgrim Core FTUE Next Button Object Reference)
   )
  ]}
 EventGraph {links [], nodes []}
]
variables []
widgets {
 info {inheritedWidgetCount=0, parentClass=/Script/PilgrimCoreUI.PilgrimFTUEWidgetBase, rootWidgetClass=&Border, widgetCount=28}
 widgets [
  Border_2 {
   widgetClassPath=&Border
   widget=Border_2
   properties {
    background {
     bIsDynamicallyLoaded=false
     drawAs=Image
     imageType=FullColor
     mirroring=NoMirror
     resourceName=None
     resourceObject=/FMCore/Materials/MI_UI_Pilgrim_Fullscreen_Background
     tiling=NoTile
     imageSize=<32 32>
     margin=<ltrb 0 0 0 0>
     outlineSettings {
      bUseBrushTransparency=false
      roundingType=HalfHeightRadius
      width=0
      color {colorUseRule=UseColor_Specified, specifiedColor=<rgba 0 0 0 0>}
      cornerRadii=<0 0 0 0>
     }
     tintColor {colorUseRule=UseColor_Specified, specifiedColor=<rgba 1 1 1 1>}
     uVRegion {bIsValid=false, max=<0 0>, min=<0 0>}
    }
    padding=<ltrb 0 0 0 0>
   }
  }
  Overlay_36 {widgetClassPath=&Overlay, parent=Border_2, widget=Overlay_36, slot {class=BorderSlot, properties {padding=<ltrb 0 0 0 0>}}}
  WBP_CaptureForPostBufferUpdate {
   widgetClassPath=&WBP_CaptureForPostBufferUpdate
   bIsVariable=true
   parent=Overlay_36
   slot {class=OverlaySlot, properties {horizontalAlignment=HAlign_Fill, verticalAlignment=VAlign_Fill}}
  }
  Border_1 {
   widgetClassPath=&Border
   parent=Overlay_36
   widget=Border_1
   properties {
    background {
     bIsDynamicallyLoaded=false
     drawAs=Image
     imageType=FullColor
     mirroring=NoMirror
     resourceName=None
     resourceObject=/FMCore/Materials/MusicLibraryBG_Inst_DublicateFMCore
     tiling=NoTile
     imageSize=<32 32>
     margin=<ltrb 0 0 0 0>
     outlineSettings {
      bUseBrushTransparency=false
      roundingType=HalfHeightRadius
      width=0
      color {colorUseRule=UseColor_Specified, specifiedColor=<rgba 0 0 0 0>}
      cornerRadii=<0 0 0 0>
     }
     tintColor {colorUseRule=UseColor_Specified, specifiedColor=<rgba 1 1 1 1>}
     uVRegion {bIsValid=false, max=<0 0>, min=<0 0>}
    }
    padding=<ltrb 0 40 0 40>
   }
   slot {class=OverlaySlot, properties {horizontalAlignment=HAlign_Fill, verticalAlignment=VAlign_Center, padding=<ltrb 0 100 0 100>}}
  }
  VerticalBox_34 {
   widgetClassPath=&VerticalBox
   parent=Border_1
   widget=VerticalBox_34
   slot {class=BorderSlot, properties {padding=<ltrb 0 40 0 40>}}
  }
  TitleTextWidget {
   widgetClassPath=&CommonTextBlock
   parent=VerticalBox_34
   widget=TitleTextWidget
   properties {
    text=HOW TO PLAY
    font {
     bForceMonospaced=false
     bMaterialIsStencil=false
     fontMaterial=null
     fontObject=&BurbankSmall
     letterSpacing=0
     monospacedWidth=1
     size=36
     skewAmount=0
     typefaceFontName=Black
     outlineSettings {
      bApplyOutlineToDropShadows=false
      bMiteredCorners=false
      bSeparateFillAlpha=false
      outlineMaterial=null
      outlineSize=0
      outlineColor=<rgba 0 0 0 1>
     }
    }
    shadowOffset=<0 0>
   }
   slot {class=VerticalBoxSlot, properties {horizontalAlignment=HAlign_Center, verticalAlignment=VAlign_Center}}
  }
  Spacer_3 {
   widgetClassPath=&Spacer
   parent=VerticalBox_34
   widget=Spacer_3
   properties {size=<1 31.02>}
   slot {class=VerticalBoxSlot, properties {verticalAlignment=VAlign_Center}}
  }
  HorizontalBox_0 {
   widgetClassPath=&HorizontalBox
   parent=VerticalBox_34
   widget=HorizontalBox_0
   slot {class=VerticalBoxSlot, properties {horizontalAlignment=HAlign_Center, verticalAlignment=VAlign_Center}}
  }
  Overlay_0 {widgetClassPath=&Overlay, parent=HorizontalBox_0, widget=Overlay_0, slot {class=HorizontalBoxSlot}}
  TutorialImage {
   widgetClassPath=&Image
   bIsVariable=true
   parent=Overlay_0
   widget=TutorialImage
   properties {brush {
     bIsDynamicallyLoaded=false
     drawAs=Image
     imageType=FullColor
     mirroring=NoMirror
     resourceName=None
     resourceObject=/PilgrimCore/Widgets/Materials/M_UI_PilgrimHowtoPlayImage
     tiling=NoTile
     imageSize=<924 576>
     margin=<ltrb 0.4 0.4 0.4 0.4>
     outlineSettings {
      bUseBrushTransparency=false
      roundingType=HalfHeightRadius
      width=0
      color {colorUseRule=UseColor_Specified, specifiedColor=<rgba 0 0 0 0>}
      cornerRadii=<0 0 0 0>
     }
     tintColor {colorUseRule=UseColor_Specified, specifiedColor=<rgba 1 1 1 1>}
     uVRegion {bIsValid=false, max=<0 0>, min=<0 0>}
    }}
   slot {class=OverlaySlot}
  }
  WBP_UI_TutorialNumber_3 {
   widgetClassPath=/PilgrimCore/Widgets/WBP_UI_TutorialNumber
   bIsVariable=true
   parent=Overlay_0
   widget=WBP_UI_TutorialNumber_3
   properties {number=3, renderTransform {angle=0, scale=<1 1>, shear=<0 0>, translation=<-157 -51>}}
   slot {class=OverlaySlot, properties {horizontalAlignment=HAlign_Center, verticalAlignment=VAlign_Bottom, padding=<ltrb 0 0 10 0>}}
  }
  WBP_UI_TutorialNumber_5 {
   widgetClassPath=&WBP_UI_TutorialNumber
   bIsVariable=true
   parent=Overlay_0
   widget=WBP_UI_TutorialNumber_5
   properties {number=2, renderTransform {angle=0, scale=<1 1>, shear=<0 0>, translation=<115 -51>}}
   slot {class=OverlaySlot, properties {horizontalAlignment=HAlign_Center, verticalAlignment=VAlign_Bottom, padding=<ltrb 0 0 10 0>}}
  }
  WBP_UI_TutorialNumber_4 {
   widgetClassPath=&WBP_UI_TutorialNumber
   bIsVariable=true
   parent=Overlay_0
   widget=WBP_UI_TutorialNumber_4
   properties {number=1, renderTransform {angle=0, scale=<1 1>, shear=<0 0>, translation=<248 -103>}}
   slot {class=OverlaySlot, properties {horizontalAlignment=HAlign_Center, verticalAlignment=VAlign_Bottom, padding=<ltrb 0 0 10 0>}}
  }
  Spacer {widgetClassPath=&Spacer, parent=HorizontalBox_0, widget=Spacer, properties {size=<120 0>}, slot {class=HorizontalBoxSlot}}
  TextBox {
   widgetClassPath=&VerticalBox
   parent=HorizontalBox_0
   widget=TextBox
   slot {class=HorizontalBoxSlot, properties {verticalAlignment=VAlign_Center}}
  }
  TutorialTipText_01 {widgetClassPath=&HorizontalBox, parent=TextBox, widget=TutorialTipText_01, slot {class=VerticalBoxSlot}}
  WBP_UI_TutorialNumber {
   widgetClassPath=&WBP_UI_TutorialNumber
   bIsVariable=true
   parent=TutorialTipText_01
   widget=WBP_UI_TutorialNumber
   properties {number=1}
   slot {class=HorizontalBoxSlot, properties {horizontalAlignment=HAlign_Left, verticalAlignment=VAlign_Top, padding=<ltrb 0 0 10 0>}}
  }
  ContentTextWidget {
   widgetClassPath=&CommonTextBlock
   parent=TutorialTipText_01
   widget=ContentTextWidget
   properties {
    text=Play notes when they hit the target area.
    wrapTextAt=600
    font {
     bForceMonospaced=false
     bMaterialIsStencil=false
     fontMaterial=null
     fontObject=&HeadingNow
     letterSpacing=0
     monospacedWidth=1
     size=18
     skewAmount=0
     typefaceFontName=75 Medium
     outlineSettings {
      bApplyOutlineToDropShadows=false
      bMiteredCorners=false
      bSeparateFillAlpha=false
      outlineMaterial=null
      outlineSize=0
      outlineColor=<rgba 0 0 0 1>
     }
    }
    shadowOffset=<0 0>
   }
   slot {class=HorizontalBoxSlot}
  }
  Spacer_1 {widgetClassPath=&Spacer, parent=TextBox, widget=Spacer_1, properties {size=<0 60>}, slot {class=VerticalBoxSlot}}
  TutorialTipText_02 {widgetClassPath=&HorizontalBox, parent=TextBox, widget=TutorialTipText_02, slot {class=VerticalBoxSlot}}
  WBP_UI_TutorialNumber_1 {
   widgetClassPath=&WBP_UI_TutorialNumber
   bIsVariable=true
   parent=TutorialTipText_02
   widget=WBP_UI_TutorialNumber_1
   properties {number=2}
   slot {class=HorizontalBoxSlot, properties {horizontalAlignment=HAlign_Left, verticalAlignment=VAlign_Top, padding=<ltrb 0 0 10 0>}}
  }
  ContentTextWidget_1 {
   widgetClassPath=&CommonTextBlock
   parent=TutorialTipText_02
   widget=ContentTextWidget_1
   properties {
    text=Hit every note to build a streak and boost your score multiplier.
    wrapTextAt=600
    font {
     bForceMonospaced=false
     bMaterialIsStencil=false
     fontMaterial=null
     fontObject=&HeadingNow
     letterSpacing=0
     monospacedWidth=1
     size=18
     skewAmount=0
     typefaceFontName=75 Medium
     outlineSettings {
      bApplyOutlineToDropShadows=false
      bMiteredCorners=false
      bSeparateFillAlpha=false
      outlineMaterial=null
      outlineSize=0
      outlineColor=<rgba 0 0 0 1>
     }
    }
    shadowOffset=<0 0>
   }
   slot {class=HorizontalBoxSlot}
  }
  Spacer_4 {widgetClassPath=&Spacer, parent=TextBox, widget=Spacer_4, properties {size=<0 60>}, slot {class=VerticalBoxSlot}}
  TutorialTipText_03 {widgetClassPath=&HorizontalBox, parent=TextBox, widget=TutorialTipText_03, slot {class=VerticalBoxSlot}}
  WBP_UI_TutorialNumber_2 {
   widgetClassPath=&WBP_UI_TutorialNumber
   bIsVariable=true
   parent=TutorialTipText_03
   widget=WBP_UI_TutorialNumber_2
   properties {number=3}
   slot {class=HorizontalBoxSlot, properties {horizontalAlignment=HAlign_Left, verticalAlignment=VAlign_Top, padding=<ltrb 0 0 10 0>}}
  }
  ContentTextWidget_2 {
   widgetClassPath=&CommonTextBlock
   parent=TutorialTipText_03
   widget=ContentTextWidget_2
   properties {
    text="Higher scores earn you more stars—up to five per song!"
    wrapTextAt=600
    font {
     bForceMonospaced=false
     bMaterialIsStencil=false
     fontMaterial=null
     fontObject=&HeadingNow
     letterSpacing=0
     monospacedWidth=1
     size=18
     skewAmount=0
     typefaceFontName=75 Medium
     outlineSettings {
      bApplyOutlineToDropShadows=false
      bMiteredCorners=false
      bSeparateFillAlpha=false
      outlineMaterial=null
      outlineSize=0
      outlineColor=<rgba 0 0 0 1>
     }
    }
    shadowOffset=<0 0>
   }
   slot {class=HorizontalBoxSlot}
  }
  Spacer_2 {widgetClassPath=&Spacer, parent=VerticalBox_34, widget=Spacer_2, properties {size=<40 40>}, slot {class=VerticalBoxSlot}}
  WBP_PilgrimCoreFTUE_NextButton {
   widgetClassPath=&/WBP_PilgrimCoreFTUE_NextButton
   bIsVariable=true
   parent=VerticalBox_34
   widget=WBP_PilgrimCoreFTUE_NextButton
   properties {minWidth=365}
   slot {class=VerticalBoxSlot, properties {horizontalAlignment=HAlign_Center, padding=<ltrb 0 0 0 40>}}
  }
 ]
}
