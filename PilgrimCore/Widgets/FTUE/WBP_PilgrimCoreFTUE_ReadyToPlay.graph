WidgetBlueprint /PilgrimCore/Widgets/FTUE/WBP_PilgrimCoreFTUE_ReadyToPlay
animations [AnimStageTech {
  binding_count=2
  length_seconds=1
  object=&self:AnimStageTech
  sequence {bindings [
    Glow_Box {
     id=#1
     possessed_class=Overlay
     tracks [{
       type=MovieSceneFloatTrack
       display_name=Render Opacity
       sections [{type=MovieSceneFloatSection, end_frame=20, start_frame=0, channels [None {type=MovieSceneScriptingFloatChannel, keys [[13,0],[20,1]]}]}]
      }]
    }
    StageTech_Image {
     id=#2
     possessed_class=Image
     tracks [
      {
       type=MovieScene2DTransformTrack
       sections [{
         type=MovieScene2DTransformSection
         end_frame=20
         start_frame=0
         channels [
          Translation.X {type=MovieSceneScriptingFloatChannel, keys [[0,-300],[20,0]]}
          Translation.Y {type=MovieSceneScriptingFloatChannel, const [0,1]}
         ]
        }]
      }
      {
       type=MovieSceneFloatTrack
       display_name=Render Opacity
       sections [{type=MovieSceneFloatSection, end_frame=20, start_frame=0, channels [None {type=MovieSceneScriptingFloatChannel, keys [[6,0],[20,1]]}]}]
      }
     ]
    }
   ]}
 }]
graphs [EventGraph {nodes [
   n1 = "On Clicked (Done_Button)"!() -> (OutputDelegate, then, Button:Common Button Base Object Reference) {is=title K2Node_ComponentBoundEvent}
   > Fire Event Element Interacted(Element=@n1.Button)
   > n2 = GetEventRouter!(execute, Scope:Object Class Reference=/Script/FortniteGame.FortPlayspace, ContextActor:Actor Object Reference=GetOwningPlayer()) -> (then, ReturnValue:Gameplay Event Router Component Object Reference)
   > Is Valid(InputObject=@n2)
   > Broadcast Event!(
    execute
    self:Gameplay Event Router Component Object Reference=@n2
    EventData:Pilgrim Event FTUE Complete Structure=Make Pilgrim Event FTUE Complete!() -> (PilgrimEvent_FTUE_Complete:Pilgrim Event FTUE Complete Structure) {is=title K2Node_MakeStruct}
    EventContext:Object Reference
   ) -> (then, ReturnValue:bool)
   Event On Activated()
   > Fire Event Main View Opened!(execute, View:Widget Object Reference=Self-Reference(), AdditionalInfo:Flow Analytics Additional Info Structure, bIsOverlay:bool=false) -> (then)
   > PlayAnimation(InAnimation=Get AnimStageTech() -> (AnimStageTech:Widget Animation Object Reference), PlaybackSpeed=2)
   Event Construct()
   > Branch(Condition=ShouldUseTouchUIForLocalPlayer!() -> (ReturnValue:bool))
   > "SetText (Text)"(
    self=Get ContentTextWidget_1() -> (ContentTextWidget_1:Common Text Object Reference)
    InText=Get TouchStageTechFeaturesText() -> (TouchStageTechFeaturesText:Text)
   )
  ]}]
variables [TouchStageTechFeaturesText {type=string, default="- Tune-up\r\n- View tips"}]
widgets {
 info {inheritedWidgetCount=0, parentClass=/Script/PilgrimCoreUI.PilgrimFTUEWidgetBase, rootWidgetClass=&Overlay, widgetCount=18}
 widgets [
  Overlay_36 {widgetClassPath=&Overlay, widget=Overlay_36}
  Border_0 {
   widgetClassPath=&Border
   parent=Overlay_36
   widget=Border_0
   properties {
    background {
     bIsDynamicallyLoaded=false
     drawAs=Image
     imageType=NoImage
     mirroring=NoMirror
     resourceName=None
     resourceObject=/FMCore/Materials/MI_UI_Pilgrim_Fullscreen_Background
     tiling=NoTile
     imageSize=<32 32>
     margin=<ltrb 0 0 0 0>
     outlineSettings {
      bUseBrushTransparency=false
      roundingType=HalfHeightRadius
      width=0
      color {colorUseRule=UseColor_Specified, specifiedColor=<rgba 0 0 0 0>}
      cornerRadii=<0 0 0 0>
     }
     tintColor {colorUseRule=UseColor_Specified, specifiedColor=<rgba 1 1 1 1>}
     uVRegion {bIsValid=false, max=<0 0>, min=<0 0>}
    }
    padding=<ltrb 0 0 0 0>
   }
   slot {class=OverlaySlot, properties {horizontalAlignment=HAlign_Fill, verticalAlignment=VAlign_Fill}}
  }
  Overlay_11 {widgetClassPath=&Overlay, parent=Border_0, widget=Overlay_11, slot {class=BorderSlot, properties {padding=<ltrb 0 0 0 0>}}}
  WBP_CaptureForPostBufferUpdate {
   widgetClassPath=&WBP_CaptureForPostBufferUpdate
   bIsVariable=true
   parent=Overlay_11
   slot {class=OverlaySlot, properties {horizontalAlignment=HAlign_Fill, verticalAlignment=VAlign_Fill}}
  }
  Content_Border {
   widgetClassPath=&Border
   parent=Overlay_11
   widget=Content_Border
   properties {
    background {
     bIsDynamicallyLoaded=false
     drawAs=Image
     imageType=NoImage
     mirroring=NoMirror
     resourceName=None
     resourceObject=/FMCore/Materials/MusicLibraryBG_Inst_DublicateFMCore
     tiling=NoTile
     imageSize=<32 32>
     margin=<ltrb 0 0 0 0>
     outlineSettings {
      bUseBrushTransparency=false
      roundingType=HalfHeightRadius
      width=0
      color {colorUseRule=UseColor_Specified, specifiedColor=<rgba 0 0 0 0>}
      cornerRadii=<0 0 0 0>
     }
     tintColor {colorUseRule=UseColor_Specified, specifiedColor=<rgba 1 1 1 1>}
     uVRegion {bIsValid=false, max=<0 0>, min=<0 0>}
    }
    padding=<ltrb 0 40 0 40>
   }
   slot {class=OverlaySlot, properties {horizontalAlignment=HAlign_Fill, verticalAlignment=VAlign_Fill, padding=<ltrb 0 100 0 100>}}
  }
  Overlay_2 {widgetClassPath=&Overlay, parent=Content_Border, widget=Overlay_2, slot {class=BorderSlot, properties {padding=<ltrb 0 40 0 40>}}}
  VerticalBox_34 {
   widgetClassPath=&VerticalBox
   parent=Overlay_2
   widget=VerticalBox_34
   slot {class=OverlaySlot, properties {horizontalAlignment=HAlign_Center, verticalAlignment=VAlign_Fill}}
  }
  TitleTextWidget {
   widgetClassPath=&CommonTextBlock
   parent=VerticalBox_34
   widget=TitleTextWidget
   properties {
    style=/FortUILibrary/Styles/Typography/Foundation/TextStyle_BurbankSmall_L_White.TextStyle_BurbankSmall_L_White_C
    text="YOU'RE READY TO PLAY!"
    font {
     bForceMonospaced=false
     bMaterialIsStencil=false
     fontMaterial=null
     fontObject=&BurbankSmall
     letterSpacing=0
     monospacedWidth=1
     size=36
     skewAmount=0
     typefaceFontName=Black
     outlineSettings {
      bApplyOutlineToDropShadows=false
      bMiteredCorners=false
      bSeparateFillAlpha=false
      outlineMaterial=null
      outlineSize=0
      outlineColor=<rgba 0 0 0 1>
     }
    }
    shadowOffset=<0 0>
   }
   slot {class=VerticalBoxSlot, properties {horizontalAlignment=HAlign_Center, verticalAlignment=VAlign_Center}}
  }
  Spacer_1 {
   widgetClassPath=&Spacer
   parent=VerticalBox_34
   widget=Spacer_1
   properties {size=<1 38>}
   slot {class=VerticalBoxSlot, properties {verticalAlignment=VAlign_Center}}
  }
  ContentTextWidget {
   widgetClassPath=&CommonTextBlock
   parent=VerticalBox_34
   widget=ContentTextWidget
   properties {
    justification=Center
    style=&TextStyle_BurbankSmall_L_White_C
    text="Remember to visit the Game Menu\r\nif you ever need to:"
    font {
     bForceMonospaced=false
     bMaterialIsStencil=false
     fontMaterial=null
     fontObject=&BurbankSmall
     letterSpacing=0
     monospacedWidth=1
     size=36
     skewAmount=0
     typefaceFontName=Black
     outlineSettings {
      bApplyOutlineToDropShadows=false
      bMiteredCorners=false
      bSeparateFillAlpha=false
      outlineMaterial=null
      outlineSize=0
      outlineColor=<rgba 0 0 0 1>
     }
    }
    shadowOffset=<0 0>
   }
   slot {class=VerticalBoxSlot, properties {horizontalAlignment=HAlign_Center, verticalAlignment=VAlign_Center}}
  }
  Spacer_4 {
   widgetClassPath=&Spacer
   parent=VerticalBox_34
   widget=Spacer_4
   properties {size=<1 38>}
   slot {class=VerticalBoxSlot, properties {verticalAlignment=VAlign_Center}}
  }
  ContentTextWidget_1 {
   widgetClassPath=&CommonTextBlock
   bIsVariable=true
   parent=VerticalBox_34
   widget=ContentTextWidget_1
   properties {
    justification=Center
    style=&TextStyle_BurbankSmall_L_White_C
    text="- Tune-up\r\n- View tips\r\n- Change controls"
    font {
     bForceMonospaced=false
     bMaterialIsStencil=false
     fontMaterial=null
     fontObject=&BurbankSmall
     letterSpacing=0
     monospacedWidth=1
     size=36
     skewAmount=0
     typefaceFontName=Black
     outlineSettings {
      bApplyOutlineToDropShadows=false
      bMiteredCorners=false
      bSeparateFillAlpha=false
      outlineMaterial=null
      outlineSize=0
      outlineColor=<rgba 0 0 0 1>
     }
    }
    shadowOffset=<0 0>
   }
   slot {class=VerticalBoxSlot, properties {horizontalAlignment=HAlign_Center, verticalAlignment=VAlign_Center}}
  }
  Done_Button {
   widgetClassPath=/FortUILibrary/UIKit/Buttons/WBP_UIKit_Button_Loud
   bIsVariable=true
   parent=VerticalBox_34
   widget=Done_Button
   properties {bRequiresHold=true, bUseNaturalNavigation=true, text="Done (Hold)"}
   slot {class=VerticalBoxSlot, properties {horizontalAlignment=HAlign_Center, verticalAlignment=VAlign_Bottom, size {sizeRule=Fill, value=1}}}
  }
  StageTech_ImageBox {
   widgetClassPath=&Overlay
   bIsVariable=true
   parent=Overlay_2
   widget=StageTech_ImageBox
   slot {class=OverlaySlot, properties {verticalAlignment=VAlign_Bottom, padding=<ltrb -200 0 40 -40>}}
  }
  Glow_Box {
   widgetClassPath=&Overlay
   parent=StageTech_ImageBox
   widget=Glow_Box
   slot {class=OverlaySlot, properties {horizontalAlignment=HAlign_Center, verticalAlignment=VAlign_Center}}
  }
  PinkGlow {
   widgetClassPath=&Image
   bIsVariable=true
   parent=Glow_Box
   widget=PinkGlow
   properties {
    brush {
     bIsDynamicallyLoaded=false
     drawAs=Image
     imageType=FullColor
     mirroring=NoMirror
     resourceName=None
     resourceObject=/FMCalibration/Materials/MI_UI_GlowTechPink
     tiling=NoTile
     imageSize=<600 600>
     margin=<ltrb 0 0 0 0>
     outlineSettings {
      bUseBrushTransparency=false
      roundingType=HalfHeightRadius
      width=0
      color {colorUseRule=UseColor_Specified, specifiedColor=<rgba 0 0 0 0>}
      cornerRadii=<0 0 0 0>
     }
     tintColor {colorUseRule=UseColor_Specified, specifiedColor=<rgba 1 1 1 0.2>}
     uVRegion {bIsValid=false, max=<0 0>, min=<0 0>}
    }
    renderTransform {angle=0, scale=<1 1>, shear=<0 0>, translation=<-50 -55>}
   }
   slot {class=OverlaySlot}
  }
  BlueGlow {
   widgetClassPath=&Image
   bIsVariable=true
   parent=Glow_Box
   widget=BlueGlow
   properties {
    brush {
     bIsDynamicallyLoaded=false
     drawAs=Image
     imageType=FullColor
     mirroring=NoMirror
     resourceName=None
     resourceObject=/FMCalibration/Materials/MI_UI_GlowTechBlue
     tiling=NoTile
     imageSize=<400 400>
     margin=<ltrb 0 0 0 0>
     outlineSettings {
      bUseBrushTransparency=false
      roundingType=HalfHeightRadius
      width=0
      color {colorUseRule=UseColor_Specified, specifiedColor=<rgba 0 0 0 0>}
      cornerRadii=<0 0 0 0>
     }
     tintColor {colorUseRule=UseColor_Specified, specifiedColor=<rgba 1 1 1 0.2>}
     uVRegion {bIsValid=false, max=<0 0>, min=<0 0>}
    }
    renderTransform {angle=0, scale=<1 1>, shear=<0 0>, translation=<180 350>}
   }
   slot {class=OverlaySlot}
  }
  StageTech_Image {
   widgetClassPath=&Image
   bIsVariable=true
   parent=StageTech_ImageBox
   widget=StageTech_Image
   properties {
    visibility=SelfHitTestInvisible
    brush {
     bIsDynamicallyLoaded=false
     drawAs=Image
     imageType=FullColor
     mirroring=NoMirror
     resourceName=None
     resourceObject=/FMCalibration/Materials/M_UI_StageTech_Art
     tiling=NoTile
     imageSize=<900 900>
     margin=<ltrb 0 0 0 0>
     outlineSettings {
      bUseBrushTransparency=false
      roundingType=HalfHeightRadius
      width=0
      color {colorUseRule=UseColor_Specified, specifiedColor=<rgba 0 0 0 0>}
      cornerRadii=<0 0 0 0>
     }
     tintColor {colorUseRule=UseColor_Specified, specifiedColor=<rgba 1 1 1 1>}
     uVRegion {bIsValid=false, max=<0 0>, min=<0 0>}
    }
   }
   slot {class=OverlaySlot}
  }
 ]
}
