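# Banner message widget for the Pilgrim first-time user experience, built as a CommonUserWidget that shows a title over a styled background.
# A horizontally filling Border with the Pilgrim fullscreen background material holds an overlay of a post-buffer capture widget, padded title text and an image.
# The rounded-box image sits bottom-right and uses negative padding to hang below the banner.
# The event graph has only unconnected stub events, and the title holds placeholder-like text.
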WidgetBlueprint /PilgrimCore/Widgets/FTUE/WBP_PilgrimFTUE_BannerMessage
graphs [EventGraph {nodes [
   Event PreConstruct()
   Event Construct()
   Event Tick()
  ]}]
variables []
widgets {
 info {inheritedWidgetCount=0, parentClass=&CommonUserWidget, rootWidgetClass=&Overlay, widgetCount=6}
 widgets [
  Overlay_36 {widgetClassPath=&Overlay, widget=Overlay_36}
  Border_0 {
   widgetClassPath=&Border
   parent=Overlay_36
   widget=Border_0
   properties {
    background {
     bIsDynamicallyLoaded=false
     drawAs=Image
     imageType=NoImage
     mirroring=NoMirror
     resourceName=None
     resourceObject=/FMCore/Materials/MI_UI_Pilgrim_Fullscreen_Background
     tiling=NoTile
     imageSize=<32 32>
     margin=<ltrb 0 0 0 0>
     outlineSettings {
      bUseBrushTransparency=false
      roundingType=HalfHeightRadius
      width=0
      color {colorUseRule=UseColor_Specified, specifiedColor=<rgba 0 0 0 0>}
      cornerRadii=<0 0 0 0>
     }
     tintColor {colorUseRule=UseColor_Specified, specifiedColor=<rgba 1 1 1 1>}
     uVRegion {bIsValid=false, max=<0 0>, min=<0 0>}
    }
    padding=<ltrb 0 0 0 0>
   }
   slot {class=OverlaySlot, properties {horizontalAlignment=HAlign_Fill, padding=<ltrb 0 50 0 0>}}
  }
  Overlay_11 {widgetClassPath=&Overlay, parent=Border_0, widget=Overlay_11, slot {class=BorderSlot, properties {padding=<ltrb 0 0 0 0>}}}
  WBP_CaptureForPostBufferUpdate {
   widgetClassPath=&WBP_CaptureForPostBufferUpdate
   bIsVariable=true
   parent=Overlay_11
   slot {class=OverlaySlot, properties {horizontalAlignment=HAlign_Fill, verticalAlignment=VAlign_Fill}}
  }
  TitleTextWidget {
   widgetClassPath=&CommonTextBlock
   parent=Overlay_11
   widget=TitleTextWidget
   properties {
    style=/FortUILibrary/Styles/Typography/Foundation/TextStyle_BurbankSmall_L_White.TextStyle_BurbankSmall_L_White_C
    text="WHAT?!?!?!"
    font {
     bForceMonospaced=false
     bMaterialIsStencil=false
     fontMaterial=null
     fontObject=&BurbankSmall
     letterSpacing=0
     monospacedWidth=1
     size=36
     skewAmount=0
     typefaceFontName=Black
     outlineSettings {
      bApplyOutlineToDropShadows=false
      bMiteredCorners=false
      bSeparateFillAlpha=false
      outlineMaterial=null
      outlineSize=0
      outlineColor=<rgba 0 0 0 1>
     }
    }
    shadowOffset=<0 0>
   }
   slot {class=OverlaySlot, properties {horizontalAlignment=HAlign_Fill, verticalAlignment=VAlign_Fill, padding=<ltrb 50 50 50 50>}}
  }
  Image_63 {
   widgetClassPath=&Image
   bIsVariable=true
   parent=Overlay_11
   widget=Image_63
   properties {brush {
     bIsDynamicallyLoaded=false
     drawAs=RoundedBox
     imageType=FullColor
     mirroring=NoMirror
     resourceName=None
     resourceObject=null
     tiling=NoTile
     imageSize=<250 250>
     margin=<ltrb 0 0 0 0>
     outlineSettings {
      bUseBrushTransparency=false
      roundingType=HalfHeightRadius
      width=0
      color {colorUseRule=UseColor_Specified, specifiedColor=<rgba 0 0 0 0>}
      cornerRadii=<0 0 0 0>
     }
     tintColor {colorUseRule=UseColor_Specified, specifiedColor=<rgba 1 1 1 1>}
     uVRegion {bIsValid=false, max=<0 0>, min=<0 0>}
    }}
   slot {class=OverlaySlot, properties {horizontalAlignment=HAlign_Right, verticalAlignment=VAlign_Bottom, padding=<ltrb 0 0 50 -75>}}
  }
 ]
}
