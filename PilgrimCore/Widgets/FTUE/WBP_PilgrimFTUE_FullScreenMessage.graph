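# Full-screen CommonUI activatable FTUE message for Pilgrim, prompting players who have not played the Main Stage.
# Stacked over a material background and post-buffer capture widget are title and paragraph text, bottom-right stage-tech art with glows, and two hold-to-confirm setlist buttons.
# AnimStageTech slides and fades in the art, then fades in its glows.
# Activated, Construct and Done_Button click events are empty stubs, so this graph neither plays the animation nor handles either choice.
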WidgetBlueprint /PilgrimCore/Widgets/FTUE/WBP_PilgrimFTUE_FullScreenMessage
animations [AnimStageTech {
  binding_count=2
  length_seconds=1
  object=&self:AnimStageTech
  sequence {bindings [
    Glow_Box {
     id=#1
     possessed_class=Overlay
     tracks [{
       type=MovieSceneFloatTrack
       display_name=Render Opacity
       sections [{type=MovieSceneFloatSection, end_frame=20, start_frame=0, channels [None {type=MovieSceneScriptingFloatChannel, keys [[13,0],[20,1]]}]}]
      }]
    }
    StageTech_Image {
     id=#2
     possessed_class=Image
     tracks [
      {
       type=MovieScene2DTransformTrack
       sections [{
         type=MovieScene2DTransformSection
         end_frame=20
         start_frame=0
         channels [
          Translation.X {type=MovieSceneScriptingFloatChannel, keys [[0,-300],[20,0]]}
          Translation.Y {type=MovieSceneScriptingFloatChannel, const [0,1]}
         ]
        }]
      }
      {
       type=MovieSceneFloatTrack
       display_name=Render Opacity
       sections [{type=MovieSceneFloatSection, end_frame=20, start_frame=0, channels [None {type=MovieSceneScriptingFloatChannel, keys [[6,0],[20,1]]}]}]
      }
     ]
    }
   ]}
 }]
graphs [EventGraph {nodes [
   "On Clicked (Done_Button)"!() -> (OutputDelegate, then, Button:Common Button Base Object Reference) {is=title K2Node_ComponentBoundEvent}
   Event On Activated()
   Event Construct()
  ]}]
variables [TouchStageTechFeaturesText {type=string, default="- Tune-up\r\n- View tips"}]
widgets {
 info {inheritedWidgetCount=0, parentClass=&CommonActivatableWidget, rootWidgetClass=&Overlay, widgetCount=19}
 widgets [
  Overlay_36 {widgetClassPath=&Overlay, widget=Overlay_36}
  Border_0 {
   widgetClassPath=&Border
   parent=Overlay_36
   widget=Border_0
   properties {
    background {
     bIsDynamicallyLoaded=false
     drawAs=Image
     imageType=NoImage
     mirroring=NoMirror
     resourceName=None
     resourceObject=/FMCore/Materials/MI_UI_Pilgrim_Fullscreen_Background
     tiling=NoTile
     imageSize=<32 32>
     margin=<ltrb 0 0 0 0>
     outlineSettings {
      bUseBrushTransparency=false
      roundingType=HalfHeightRadius
      width=0
      color {colorUseRule=UseColor_Specified, specifiedColor=<rgba 0 0 0 0>}
      cornerRadii=<0 0 0 0>
     }
     tintColor {colorUseRule=UseColor_Specified, specifiedColor=<rgba 1 1 1 1>}
     uVRegion {bIsValid=false, max=<0 0>, min=<0 0>}
    }
    padding=<ltrb 0 0 0 0>
   }
   slot {class=OverlaySlot, properties {horizontalAlignment=HAlign_Fill, verticalAlignment=VAlign_Fill}}
  }
  Overlay_11 {widgetClassPath=&Overlay, parent=Border_0, widget=Overlay_11, slot {class=BorderSlot, properties {padding=<ltrb 0 0 0 0>}}}
  WBP_CaptureForPostBufferUpdate {
   widgetClassPath=&WBP_CaptureForPostBufferUpdate
   bIsVariable=true
   parent=Overlay_11
   slot {class=OverlaySlot, properties {horizontalAlignment=HAlign_Fill, verticalAlignment=VAlign_Fill}}
  }
  StageTech_ImageBox {
   widgetClassPath=&Overlay
   bIsVariable=true
   parent=Overlay_11
   widget=StageTech_ImageBox
   slot {class=OverlaySlot, properties {horizontalAlignment=HAlign_Right, verticalAlignment=VAlign_Bottom, padding=<ltrb 0 0 -200 0>}}
  }
  Glow_Box {
   widgetClassPath=&Overlay
   parent=StageTech_ImageBox
   widget=Glow_Box
   slot {class=OverlaySlot, properties {horizontalAlignment=HAlign_Center, verticalAlignment=VAlign_Center}}
  }
  PinkGlow {
   widgetClassPath=&Image
   bIsVariable=true
   parent=Glow_Box
   widget=PinkGlow
   properties {
    brush {
     bIsDynamicallyLoaded=false
     drawAs=Image
     imageType=FullColor
     mirroring=NoMirror
     resourceName=None
     resourceObject=/FMCalibration/Materials/MI_UI_GlowTechPink
     tiling=NoTile
     imageSize=<600 600>
     margin=<ltrb 0 0 0 0>
     outlineSettings {
      bUseBrushTransparency=false
      roundingType=HalfHeightRadius
      width=0
      color {colorUseRule=UseColor_Specified, specifiedColor=<rgba 0 0 0 0>}
      cornerRadii=<0 0 0 0>
     }
     tintColor {colorUseRule=UseColor_Specified, specifiedColor=<rgba 1 1 1 0.2>}
     uVRegion {bIsValid=false, max=<0 0>, min=<0 0>}
    }
    renderTransform {angle=0, scale=<1 1>, shear=<0 0>, translation=<-50 -55>}
   }
   slot {class=OverlaySlot}
  }
  BlueGlow {
   widgetClassPath=&Image
   bIsVariable=true
   parent=Glow_Box
   widget=BlueGlow
   properties {
    brush {
     bIsDynamicallyLoaded=false
     drawAs=Image
     imageType=FullColor
     mirroring=NoMirror
     resourceName=None
     resourceObject=/FMCalibration/Materials/MI_UI_GlowTechBlue
     tiling=NoTile
     imageSize=<400 400>
     margin=<ltrb 0 0 0 0>
     outlineSettings {
      bUseBrushTransparency=false
      roundingType=HalfHeightRadius
      width=0
      color {colorUseRule=UseColor_Specified, specifiedColor=<rgba 0 0 0 0>}
      cornerRadii=<0 0 0 0>
     }
     tintColor {colorUseRule=UseColor_Specified, specifiedColor=<rgba 1 1 1 0.2>}
     uVRegion {bIsValid=false, max=<0 0>, min=<0 0>}
    }
    renderTransform {angle=0, scale=<1 1>, shear=<0 0>, translation=<180 350>}
   }
   slot {class=OverlaySlot}
  }
  StageTech_Image {
   widgetClassPath=&Image
   bIsVariable=true
   parent=StageTech_ImageBox
   widget=StageTech_Image
   properties {
    visibility=SelfHitTestInvisible
    brush {
     bIsDynamicallyLoaded=false
     drawAs=Image
     imageType=FullColor
     mirroring=NoMirror
     resourceName=None
     resourceObject=/FMCalibration/Materials/MI_UI_StageTech_Art
     tiling=NoTile
     imageSize=<900 900>
     margin=<ltrb 0 0 0 0>
     outlineSettings {
      bUseBrushTransparency=false
      roundingType=HalfHeightRadius
      width=0
      color {colorUseRule=UseColor_Specified, specifiedColor=<rgba 0 0 0 0>}
      cornerRadii=<0 0 0 0>
     }
     tintColor {colorUseRule=UseColor_Specified, specifiedColor=<rgba 1 1 1 1>}
     uVRegion {bIsValid=false, max=<0 0>, min=<0 0>}
    }
   }
   slot {class=OverlaySlot, properties {verticalAlignment=VAlign_Bottom}}
  }
  VerticalBox_1 {
   widgetClassPath=&VerticalBox
   parent=Overlay_11
   widget=VerticalBox_1
   slot {class=OverlaySlot, properties {horizontalAlignment=HAlign_Fill, verticalAlignment=VAlign_Fill, padding=<ltrb 0 100 0 100>}}
  }
  Overlay_2 {
   widgetClassPath=&Overlay
   parent=VerticalBox_1
   widget=Overlay_2
   slot {class=VerticalBoxSlot, properties {padding=<ltrb 100 0 100 0>, size {sizeRule=Fill, value=1}}}
  }
  VerticalBox_34 {
   widgetClassPath=&VerticalBox
   parent=Overlay_2
   widget=VerticalBox_34
   slot {class=OverlaySlot, properties {horizontalAlignment=HAlign_Fill, verticalAlignment=VAlign_Center}}
  }
  TitleTextWidget {
   widgetClassPath=&CommonTextBlock
   parent=VerticalBox_34
   widget=TitleTextWidget
   properties {
    style=/FortUILibrary/Styles/Typography/Foundation/TextStyle_BurbankSmall_L_White.TextStyle_BurbankSmall_L_White_C
    text="WHAT?!?!?!"
    font {
     bForceMonospaced=false
     bMaterialIsStencil=false
     fontMaterial=null
     fontObject=&BurbankSmall
     letterSpacing=0
     monospacedWidth=1
     size=36
     skewAmount=0
     typefaceFontName=Black
     outlineSettings {
      bApplyOutlineToDropShadows=false
      bMiteredCorners=false
      bSeparateFillAlpha=false
      outlineMaterial=null
      outlineSize=0
      outlineColor=<rgba 0 0 0 1>
     }
    }
    shadowOffset=<0 0>
   }
   slot {class=VerticalBoxSlot, properties {horizontalAlignment=HAlign_Left, verticalAlignment=VAlign_Center, padding=<ltrb 0 38 0 0>}}
  }
  Paragraph1 {
   widgetClassPath=&CommonTextBlock
   parent=VerticalBox_34
   widget=Paragraph1
   properties {
    bAutoCollapseWithEmptyText=true
    style=&TextStyle_BurbankSmall_L_White_C
    text=You haven't played the Main Stage yet???
    wrapTextAt=1000
    font {
     bForceMonospaced=false
     bMaterialIsStencil=false
     fontMaterial=null
     fontObject=&BurbankSmall
     letterSpacing=0
     monospacedWidth=1
     size=36
     skewAmount=0
     typefaceFontName=Black
     outlineSettings {
      bApplyOutlineToDropShadows=false
      bMiteredCorners=false
      bSeparateFillAlpha=false
      outlineMaterial=null
      outlineSize=0
      outlineColor=<rgba 0 0 0 1>
     }
    }
    shadowOffset=<0 0>
   }
   slot {class=VerticalBoxSlot, properties {horizontalAlignment=HAlign_Left, verticalAlignment=VAlign_Center, padding=<ltrb 0 38 0 0>}}
  }
  Paragraph2 {
   widgetClassPath=&CommonTextBlock
   parent=VerticalBox_34
   widget=Paragraph2
   properties {
    style=&TextStyle_BurbankSmall_L_White_C
    font {
     bForceMonospaced=false
     bMaterialIsStencil=false
     fontMaterial=null
     fontObject=&BurbankSmall
     letterSpacing=0
     monospacedWidth=1
     size=36
     skewAmount=0
     typefaceFontName=Black
     outlineSettings {
      bApplyOutlineToDropShadows=false
      bMiteredCorners=false
      bSeparateFillAlpha=false
      outlineMaterial=null
      outlineSize=0
      outlineColor=<rgba 0 0 0 1>
     }
    }
    shadowOffset=<0 0>
   }
   slot {class=VerticalBoxSlot, properties {horizontalAlignment=HAlign_Left, verticalAlignment=VAlign_Center, padding=<ltrb 0 38 0 0>}}
  }
  HorizontalBox_0 {
   widgetClassPath=&HorizontalBox
   parent=VerticalBox_1
   widget=HorizontalBox_0
   slot {class=VerticalBoxSlot, properties {horizontalAlignment=HAlign_Center, verticalAlignment=VAlign_Bottom, size {sizeRule=Fill, value=0.1}}}
  }
  Done_Button {
   widgetClassPath=/FortUILibrary/UIKit/Buttons/WBP_UIKit_Button_Loud
   bIsVariable=true
   parent=HorizontalBox_0
   widget=Done_Button
   properties {bRequiresHold=true, bUseNaturalNavigation=true, text=Play Dill Bill's Setlist}
   slot {class=HorizontalBoxSlot}
  }
  Spacer_2 {
   widgetClassPath=&Spacer
   parent=HorizontalBox_0
   widget=Spacer_2
   properties {size=<50 0>}
   slot {class=HorizontalBoxSlot, properties {verticalAlignment=VAlign_Center}}
  }
  Done_Button_1 {
   widgetClassPath=&WBP_UIKit_Button_Loud
   bIsVariable=true
   parent=HorizontalBox_0
   widget=Done_Button_1
   properties {bRequiresHold=true, bUseNaturalNavigation=true, text=Make Your Own Setlist}
   slot {class=HorizontalBoxSlot}
  }
 ]
}
