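# Shared fullscreen background widget for the FMCalibration UI, built on CommonUserWidget with FTUE-themed materials.
# A root Border draws an animated-style sparks material behind an Overlay of material-driven Images.
# The Overlay stacks a collapsed Smoke atmospherics layer, two inset radial glows sharing one material for a denser lower glow, and blue bottom and top edge glows.
# It is purely visual, with empty event stubs and no logic or variables.
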WidgetBlueprint /FMCalibration/Widgets/WBP_FMCalibration_SharedBackground
graphs [EventGraph {nodes [
   Event PreConstruct()
   Event Construct()
   Event Tick()
  ]}]
variables []
widgets {
 info {inheritedWidgetCount=0, parentClass=&CommonUserWidget, rootWidgetClass=&Border, widgetCount=7}
 widgets [
  Border_1 {
   widgetClassPath=&Border
   widget=Border_1
   properties {
    background {
     bIsDynamicallyLoaded=false
     drawAs=Image
     imageType=FullColor
     mirroring=NoMirror
     resourceName=None
     resourceObject=/FMCalibration/Materials/MI_UI_FTUE_Sparks_Fullscreen_Background
     tiling=NoTile
     imageSize=<32 32>
     margin=<ltrb 0 0 0 0>
     outlineSettings {
      bUseBrushTransparency=false
      roundingType=HalfHeightRadius
      width=0
      color {colorUseRule=UseColor_Specified, specifiedColor=<rgba 0 0 0 0>}
      cornerRadii=<0 0 0 0>
     }
     tintColor {colorUseRule=UseColor_Specified, specifiedColor=<rgba 1 1 1 1>}
     uVRegion {bIsValid=false, max=<0 0>, min=<0 0>}
    }
    padding=<ltrb 0 0 0 0>
   }
  }
  Overlay_0 {widgetClassPath=&Overlay, parent=Border_1, widget=Overlay_0, slot {class=BorderSlot, properties {padding=<ltrb 0 0 0 0>}}}
  Smoke {
   widgetClassPath=&Image
   bIsVariable=true
   parent=Overlay_0
   widget=Smoke
   properties {
    visibility=Collapsed
    brush {
     bIsDynamicallyLoaded=false
     drawAs=Image
     imageType=FullColor
     mirroring=NoMirror
     resourceName=None
     resourceObject=/FMCalibration/Materials/MI_Ui_FTUE_RadiatingAtmospherics
     tiling=NoTile
     imageSize=<32 32>
     margin=<ltrb 0 0 0 0>
     outlineSettings {
      bUseBrushTransparency=false
      roundingType=HalfHeightRadius
      width=0
      color {colorUseRule=UseColor_Specified, specifiedColor=<rgba 0 0 0 0>}
      cornerRadii=<0 0 0 0>
     }
     tintColor {colorUseRule=UseColor_Specified, specifiedColor=<rgba 1 1 1 1>}
     uVRegion {bIsValid=false, max=<0 0>, min=<0 0>}
    }
   }
   slot {class=OverlaySlot, properties {horizontalAlignment=HAlign_Fill, verticalAlignment=VAlign_Fill}}
  }
  Radial_Glow {
   widgetClassPath=&Image
   bIsVariable=true
   parent=Overlay_0
   widget=Radial_Glow
   properties {brush {
     bIsDynamicallyLoaded=false
     drawAs=Image
     imageType=FullColor
     mirroring=NoMirror
     resourceName=None
     resourceObject=/FMCalibration/Materials/MI_UI_FTUE_AlphaComposite_Glow_NoFade_Blue_Bottom
     tiling=NoTile
     imageSize=<32 32>
     margin=<ltrb 0 0 0 0>
     outlineSettings {
      bUseBrushTransparency=false
      roundingType=HalfHeightRadius
      width=0
      color {colorUseRule=UseColor_Specified, specifiedColor=<rgba 0 0 0 0>}
      cornerRadii=<0 0 0 0>
     }
     tintColor {colorUseRule=UseColor_Specified, specifiedColor=<rgba 1 1 1 1>}
     uVRegion {bIsValid=false, max=<0 0>, min=<0 0>}
    }}
   slot {class=OverlaySlot, properties {horizontalAlignment=HAlign_Fill, verticalAlignment=VAlign_Fill, padding=<ltrb 400 500 400 0>}}
  }
  Radial_Glow_1 {
   widgetClassPath=&Image
   bIsVariable=true
   parent=Overlay_0
   widget=Radial_Glow_1
   properties {brush {
     bIsDynamicallyLoaded=false
     drawAs=Image
     imageType=FullColor
     mirroring=NoMirror
     resourceName=None
     resourceObject=&MI_UI_FTUE_AlphaComposite_Glow_NoFade_Blue_Bottom
     tiling=NoTile
     imageSize=<32 32>
     margin=<ltrb 0 0 0 0>
     outlineSettings {
      bUseBrushTransparency=false
      roundingType=HalfHeightRadius
      width=0
      color {colorUseRule=UseColor_Specified, specifiedColor=<rgba 0 0 0 0>}
      cornerRadii=<0 0 0 0>
     }
     tintColor {colorUseRule=UseColor_Specified, specifiedColor=<rgba 1 1 1 1>}
     uVRegion {bIsValid=false, max=<0 0>, min=<0 0>}
    }}
   slot {class=OverlaySlot, properties {horizontalAlignment=HAlign_Fill, verticalAlignment=VAlign_Fill, padding=<ltrb 500 600 500 0>}}
  }
  BottomGlow {
   widgetClassPath=&Image
   bIsVariable=true
   parent=Overlay_0
   widget=BottomGlow
   properties {brush {
     bIsDynamicallyLoaded=false
     drawAs=Image
     imageType=FullColor
     mirroring=NoMirror
     resourceName=None
     resourceObject=/FMCalibration/Materials/MI_UI_FTUE_AlphaComposite_Glow_Blue_Bottom
     tiling=NoTile
     imageSize=<1400 32>
     margin=<ltrb 0 0 0 0>
     outlineSettings {
      bUseBrushTransparency=false
      roundingType=HalfHeightRadius
      width=0
      color {colorUseRule=UseColor_Specified, specifiedColor=<rgba 0 0 0 0>}
      cornerRadii=<0 0 0 0>
     }
     tintColor {colorUseRule=UseColor_Specified, specifiedColor=<rgba 1 1 1 1>}
     uVRegion {bIsValid=false, max=<0 0>, min=<0 0>}
    }}
   slot {class=OverlaySlot, properties {horizontalAlignment=HAlign_Fill, verticalAlignment=VAlign_Fill}}
  }
  TopGlow {
   widgetClassPath=&Image
   bIsVariable=true
   parent=Overlay_0
   widget=TopGlow
   properties {brush {
     bIsDynamicallyLoaded=false
     drawAs=Image
     imageType=FullColor
     mirroring=NoMirror
     resourceName=None
     resourceObject=/FMCalibration/Materials/MI_UI_FTUE_AlphaComposite_Glow1_Blue_Top
     tiling=NoTile
     imageSize=<1400 32>
     margin=<ltrb 0 0 0 0>
     outlineSettings {
      bUseBrushTransparency=false
      roundingType=HalfHeightRadius
      width=0
      color {colorUseRule=UseColor_Specified, specifiedColor=<rgba 0 0 0 0>}
      cornerRadii=<0 0 0 0>
     }
     tintColor {colorUseRule=UseColor_Specified, specifiedColor=<rgba 1 1 1 1>}
     uVRegion {bIsValid=false, max=<0 0>, min=<0 0>}
    }}
   slot {class=OverlaySlot, properties {horizontalAlignment=HAlign_Center, verticalAlignment=VAlign_Fill}}
  }
 ]
}
