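# FTUE control scheme selector widget: a button with left and right chevrons around an input-type icon and a text label.
# Arrow and background clicks are relayed as Left, Right and Center event dispatchers for outside logic to handle.
# On construct and on CommonInput method changes, the icon switches between keyboard, gamepad and touch textures.
# Hover animates the label's font material via Trans_Hovered; SetSelector has no executed logic, and nothing here sets the label text.
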WidgetBlueprint /PilgrimCore/Widgets/FTUE/WBP_Pilgrim_ControlSchemesSelector
animations [AnimOnHover {
  binding_count=1
  length_seconds=0.25
  object=&self:AnimOnHover
  sequence {bindings [SliderValue {
     id=#1
     possessed_class=CommonTextBlock
     tracks [{
       type=MovieSceneWidgetMaterialTrack
       display_name=Font.FontMaterial
       sections [{
         type=MovieSceneParameterSection
         end_frame=5
         start_frame=0
         channels [Trans_Hovered {type=MovieSceneScriptingFloatChannel, keys [[0,0],[5,1]]}]
        }]
      }]
    }]}
 }]
graphs [
 EventGraph {nodes [
   "On Clicked (ArrowLeft)"!() -> (OutputDelegate, then, Button:Common Button Base Object Reference) {is=title K2Node_ComponentBoundEvent}
   > Call On Clicked Left!(execute:Exec, self:WBP Pilgrim Control Schemes Selector Object Reference) -> (then:Exec) {is=title K2Node_CallDelegate}
   "On Clicked (ArrowRight)"!() -> (OutputDelegate, then, Button:Common Button Base Object Reference) {is=title K2Node_ComponentBoundEvent}
   > Call On Clicked Right!(execute:Exec, self:WBP Pilgrim Control Schemes Selector Object Reference) -> (then:Exec) {is=title K2Node_CallDelegate}
   "On Clicked (ButtonBG)"!() -> (OutputDelegate, then, Button:Common Button Base Object Reference) {is=title K2Node_ComponentBoundEvent}
   > Call On Clicked Center!(execute:Exec, self:WBP Pilgrim Control Schemes Selector Object Reference) -> (then:Exec) {is=title K2Node_CallDelegate}
   WhenHoveredFocused!() -> (OutputDelegate, then, Button:Common Button Base Object Reference) {is=title K2Node_CustomEvent}
   n1 = Get AnimOnHover() -> (AnimOnHover:Widget Animation Object Reference)
   > PlayAnimationForward(InAnimation=@n1)
   WhenUnhoveredUnfocused!() -> (OutputDelegate, then, Button:Common Button Base Object Reference) {is=title K2Node_CustomEvent}
   > PlayAnimationReverse(InAnimation=@n1)
   Event Construct()
   n2 = Get ButtonBG() -> (ButtonBG:WBP UIKit Button Generic Object Reference)
   > Bind Event to On Hovered(self:Common Button Base Object Reference=@n2, Delegate=Create Event()) {is=title K2Node_AddDelegate}
   > Bind Event to On Unhovered(self:Common Button Base Object Reference=@n2, Delegate=Create Event()) {is=title K2Node_AddDelegate}
   n3 = Get CommonInputSubsystem!(PlayerController:Player Controller Object Reference=GetOwningPlayer()) -> (ReturnValue:Common Input Object Reference) {is=title K2Node_GetSubsystemFromPC}
   > Bind Event to On Input Method Changed(self=@n3, Delegate=Create Event())
   > OnInputMethodChanged!(execute, self, InputMethod:ECommonInputType Enum=GetCurrentInputType(self=@n3)) -> (then)
  ]}
 OnClickedCenter {nodes [OnClickedCenter() {is=title K2Node_FunctionEntry}]}
 OnClickedLeft {nodes [OnClickedLeft() {is=title K2Node_FunctionEntry}]}
 OnClickedRight {nodes [OnClickedRight() {is=title K2Node_FunctionEntry}]}
 OnInputMethodChanged {nodes [
   n1 = OnInputMethodChanged!() -> (then, InputMethod:ECommonInputType Enum) {is=title K2Node_FunctionEntry}
   > SetBrushFromTexture!(
    execute
    self:Image Object Reference=Get InputType_Image() -> (InputType_Image:Image Object Reference)
    Texture:Texture 2D Object Reference=Select!(
     MouseAndKeyboard:Texture 2D Object Reference=/Game/UI/Icons/Icon-Settings/T-Icon-Keyboard-64
     Gamepad:Texture 2D Object Reference=/FortUILibrary/Textures/Icons/T_UI_Icon_UIKit_ControllerMapping
     Touch:Texture 2D Object Reference=/Game/UI/Icons/Icon-Settings/T-Icon-TouchMotion-64
     Count:Texture 2D Object Reference
     Index:ECommonInputType Enum=@n1
    ) -> (ReturnValue:Texture 2D Object Reference)
    bMatchSize:bool=false
   ) -> (then)
  ]}
 SetSelector {nodes [
   SetSelector() {is=title K2Node_FunctionEntry}
   Get SettingsListEntry_Discrete() -> (SettingsListEntry_Discrete:Settings List Entry Discrete Object Reference)
  ]}
]
variables [
 OnClickedCenter {}
 OnClickedLeft {}
 OnClickedRight {}
]
widgets {
 info {inheritedWidgetCount=0, namedSlotCount=1, parentClass=&UserWidget, rootWidgetClass=&SizeBox, widgetCount=8}
 widgets [
  SizeBox {widgetClassPath=&SizeBox, bIsVariable=true, widget=SizeBox, properties {bOverride_WidthOverride=true, widthOverride=408}}
  ButtonBG {
   widgetClassPath=&WBP_UIKit_Button_Generic
   bIsVariable=true
   parent=SizeBox
   widget=ButtonBG
   properties {bHideInputAction=true, text="", useIconBlock=false}
   slot {class=SizeBoxSlot}
  }
  HorizontalBox_0 {widgetClassPath=&HorizontalBox, namedSlotHost=ButtonBG, widget=HorizontalBox_0}
  ArrowLeft {
   widgetClassPath=&WBP_UIKit_Button_Generic
   bIsVariable=true
   parent=HorizontalBox_0
   widget=ArrowLeft
   properties {
    bHideInputAction=true
    bIsFocusable=false
    icon=/FortUILibrary/Textures/Icons/T_UI_Icon_UIKit_Chevron
    iconMaterial=/FMCalibration/Materials/MI_UIKit_GenericIcon_Regular
    text=""
    useBackgroundBlock=false
    useTextBlock=false
    renderTransform {angle=0, scale=<-1 1>, shear=<0 0>, translation=<0 0>}
   }
   slot {class=HorizontalBoxSlot, properties {horizontalAlignment=HAlign_Left}}
  }
  Content_HorizontalBox {
   widgetClassPath=&HorizontalBox
   parent=HorizontalBox_0
   widget=Content_HorizontalBox
   slot {class=HorizontalBoxSlot, properties {horizontalAlignment=HAlign_Center, verticalAlignment=VAlign_Center, size {sizeRule=Fill, value=1}}}
  }
  InputType_Image {
   widgetClassPath=&Image
   bIsVariable=true
   parent=Content_HorizontalBox
   widget=InputType_Image
   properties {colorAndOpacity=<rgba 0.0048 0.004 0.0116 1>}
   slot {class=HorizontalBoxSlot, properties {horizontalAlignment=HAlign_Left, verticalAlignment=VAlign_Center}}
  }
  SliderValue {
   widgetClassPath=&CommonTextBlock
   bIsVariable=true
   parent=Content_HorizontalBox
   widget=SliderValue
   properties {
    justification=Center
    scrollStyle=&TextScrollStyle-Base_C
    text=Dummy Text
    colorAndOpacity {colorUseRule=UseColor_Specified, specifiedColor=<rgba 0.0048 0.004 0.0116 1>}
    font {
     bForceMonospaced=false
     bMaterialIsStencil=false
     fontMaterial=/FMCalibration/Materials/MI_UIKit_GenericText_Transition
     fontObject=&BurbankSmall
     letterSpacing=0
     monospacedWidth=1
     size=16
     skewAmount=0
     typefaceFontName=Bold
     outlineSettings {
      bApplyOutlineToDropShadows=false
      bMiteredCorners=false
      bSeparateFillAlpha=false
      outlineMaterial=null
      outlineSize=0
      outlineColor=<rgba 0 0 0 1>
     }
    }
    shadowOffset=<0 0>
   }
   slot {
    class=HorizontalBoxSlot
    properties {horizontalAlignment=HAlign_Left, verticalAlignment=VAlign_Center, padding=<ltrb 16 0 0 0>, size {sizeRule=Fill, value=1}}
   }
  }
  ArrowRight {
   widgetClassPath=&WBP_UIKit_Button_Generic
   bIsVariable=true
   parent=HorizontalBox_0
   widget=ArrowRight
   properties {
    bHideInputAction=true
    bIsFocusable=false
    icon=&T_UI_Icon_UIKit_Chevron
    iconMaterial=&MI_UIKit_GenericIcon_Regular
    text=""
    useBackgroundBlock=false
    useTextBlock=false
   }
   slot {class=HorizontalBoxSlot, properties {horizontalAlignment=HAlign_Right}}
  }
 ]
}
