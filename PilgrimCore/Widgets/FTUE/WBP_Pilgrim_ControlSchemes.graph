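# Pilgrim gameplay controls screen, built on PilgrimPresetPreviewer, where players preview and apply input presets for 4-lane (Easy to Hard) or 5-lane (Expert) play, including during FTUE.
# Tab, selector and input-device changes refresh key icons from preset mapping contexts, or player mappings for Custom.
# Confirming applies the preset and sends analytics; a separate button opens Fortnite's binding settings.
# FTUE state swaps the bottom buttons, and a valid quickplay-began event closes the screen.
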WidgetBlueprint /PilgrimCore/Widgets/FTUE/WBP_Pilgrim_ControlSchemes
graphs [
 BP_GetDesiredFocusTarget {nodes [
   Get Desired Focus Target()
   > Return Node!(
    execute
    ReturnValue:Widget Object Reference=Get WBP_Pilgrim_ControlSchemesSelector() -> (WBP_Pilgrim_ControlSchemesSelector:WBP Pilgrim Control Schemes Selector Object Reference)
   )
  ]}
 Buttons Nav Left {nodes [
   Buttons Nav Left!() -> (then, Navigation:EUINavigation Enum) {is=title K2Node_FunctionEntry}
   > ClickLeft()
   > Return Node()
  ]}
 Buttons Nav Right {nodes [
   Buttons Nav Right!() -> (then, Navigation:EUINavigation Enum) {is=title K2Node_FunctionEntry}
   > ClickRight()
   > Return Node()
  ]}
 ClickLeft {nodes [
   ClickLeft() {is=title K2Node_FunctionEntry}
   n1 = Get ControlsIndex() -> (ControlsIndex:int=0)
   > Decrement Int(Value=@n1)
   > n2 = Branch(Condition="Equal (Integer)"(A=@n1, B=-1))
   > n3 = Set ControlsIndex(ControlsIndex:int=2) -> (Output_Get:int=0)
   SetControlsVisuals(execute=[@n3.then, @n2.else])
   > SetInputMappings()
   > SetPresetName()
  ]}
 ClickRight {nodes [
   ClickRight() {is=title K2Node_FunctionEntry}
   n1 = Get ControlsIndex() -> (ControlsIndex=0)
   > Increment Int(Value=@n1)
   > n2 = Branch(Condition="Equal (Integer)"(A=@n1, B=3))
   > n3 = Set ControlsIndex(ControlsIndex=0) -> (Output_Get=0)
   SetControlsVisuals(execute=[@n3.then, @n2.else])
   > SetInputMappings()
   > SetPresetName()
  ]}
 CloseScreen {nodes [
   CloseScreen() {is=title K2Node_FunctionEntry}
   > RemoveScene!(
    execute
    self:Dynamic UIManager Object Reference=Get DynamicUIManager() -> (ReturnValue:Dynamic UIManager Object Reference) {is=title K2Node_GetSubsystem}
    Scene:Dynamic UIScene Object Reference=/PilgrimCore/Widgets/Pilgrim_ControlSchemes
    Player:"Player Controller Object Reference (by ref)"=GetOwningPlayer()
   ) -> (then)
  ]}
 ConfirmAndApplyPreset {nodes [
   ConfirmAndApplyPreset() {is=title K2Node_FunctionEntry}
   n1 = GetDeviceType!() -> (Device Type:EHardwareDevicePrimaryType Enum=Unspecified)
   n2 = GetDeviceType() -> (Device Type=Unspecified)
   n3 = Select!(
    Option 0:int=0
    Option 1:int=2
    Option 2:int=SelectInt(A=4, B=3, bPickA="Equal (Enum)"(A:EHardwareDevicePrimaryType Enum=@n2, B:EHardwareDevicePrimaryType Enum=KeyboardAndMouse))
    Index:int=Get ControlsIndex() -> (ControlsIndex=0)
   ) -> (ReturnValue:int)
   n4 = Select!(Option 0=4, Option 1=5, Index=Get TabIndex() -> (TabIndex:int=0)) -> (ReturnValue)
   > n5 = ApplyPreset!(
    execute
    self:Pilgrim Preset Previewer Object Reference
    CurrentPlayer:Player Controller Object Reference=GetOwningPlayer()
    SelectedNumLanes:int=@n4
    DeviceType:EHardwareDevicePrimaryType Enum=@n1
    SelectedPresetIndex:int=@n3
   ) -> (then, ReturnValue:bool)
   > n6 = Branch(Condition=@n5)
   > SendApplyControlsPresetAnalytics!(execute, self, Num Lanes:int=@n4, Device Type:EHardwareDevicePrimaryType Enum=@n1, Selected Preset Index:int=@n3) -> (then)
   > Return Node!(execute, PresetApplied:bool=true)
   FortShippingLog!(execute=@n6.else, WorldContextObject:Object Reference, InString:String=Failed to apply selected input preset, bLogAsWarning:bool=false) -> (then)
   > Return Node(PresetApplied=false)
  ]}
 EventGraph {nodes [
   n1 = "On Tab Selected (WBP_UIKit_Tabs_TopNavBar)"!() -> (OutputDelegate, then, TabId:Name=None) {is=title K2Node_ComponentBoundEvent}
   > Fire Event Sub View Opened!(
    execute
    View:Widget Object Reference=Self-Reference()
    AdditionalInfo:Flow Analytics Additional Info Structure="Add String (Non-Localized)"!(
     Info:"Flow Analytics Additional Info Structure (by ref)"=Make Flow Analytics Additional Info!() -> (FlowAnalyticsAdditionalInfo:Flow Analytics Additional Info Structure) {is=title K2Node_MakeStruct}
     Key:String=TabId
     Value:String="To String (Name)"(InName=@n1.TabId)
    ) -> (ReturnValue:Flow Analytics Additional Info Structure)
    bIsOverlay:bool=false
   ) -> (then)
   > Set TabIndex(TabIndex:int=Select(Option 0=1, Option 1=0, Index:bool="Equal (Name)"(A=@n1.TabId, B=0))) -> (Output_Get:int=0)
   > GetAppliedPreset()
   > SetCurrentMappingContext()
   > SetControlsVisuals()
   > SetInputMappings()
   > SetPresetName()
   "On Clicked Right (WBP_Pilgrim_ControlSchemesSelector)"() {is=title K2Node_ComponentBoundEvent}
   > FireEventElementInteractedByName!(execute, Element Name:String=OnClicked_WBP_Pilgrim_ControlSchemesSelector_Right, WidgetCallingFrom:Widget Object Reference=Self-Reference()) -> (then, Success:bool)
   > ClickRight() {is=title K2Node_CallFunction}
   "On Clicked Left (WBP_Pilgrim_ControlSchemesSelector)"() {is=title K2Node_ComponentBoundEvent}
   > FireEventElementInteractedByName(Element Name=OnClicked_WBP_Pilgrim_ControlSchemesSelector_Left, WidgetCallingFrom=Self-Reference())
   > ClickLeft() {is=title K2Node_CallFunction}
   n2 = "On Clicked (CustomControl_Button)"!() -> (OutputDelegate, then, Button:Common Button Base Object Reference) {is=title K2Node_ComponentBoundEvent}
   > Fire Event Element Interacted(Element=@n2.Button)
   > OpenSettings()
   "On Clicked Center (WBP_Pilgrim_ControlSchemesSelector)"() {is=title K2Node_ComponentBoundEvent}
   > n3 = ConfirmAndApplyPreset!(execute, self) -> (then, PresetApplied:bool) {is=title K2Node_CallFunction}
   n4 = Get IsFTUE() -> (IsFTUE:bool)
   > Branch(Condition=AND Boolean(A=@n3, B=NOT Boolean(A=@n4)))
   > CloseScreen() {is=title K2Node_CallFunction}
   n5 = "On Clicked (Next_FTUE_Button)"!() -> (OutputDelegate, then, Button:Common Button Base Object Reference) {is=title K2Node_ComponentBoundEvent}
   > Fire Event Element Interacted(Element=@n5.Button)
   > n6 = ConfirmAndApplyPreset()
   > n7 = Branch(Condition=@n6)
   CloseScreen(execute=[@n8.then, @n7.then])
   n8 = FortShippingLog(
    execute=@n7.else
    WorldContextObject=GetOwningPlayer()
    InString="WBP_Pilgrim_ControlSchemes::OnClicked (Next_FTUE_Button) -- Failed to save preset."
    bLogAsWarning=false
   )
   "On Pressed (TouchArea_Right)"() {is=title K2Node_ComponentBoundEvent}
   > FireEventElementInteractedByName(Element Name=OnPressed_TouchArea_Right, WidgetCallingFrom=Self-Reference())
   n9 = Get WBP_UIKit_Tabs_TopNavBar() -> (WBP_UIKit_Tabs_TopNavBar:WBP UIKit Tabs Top Nav Bar Object Reference)
   > SelectNextTab!(execute, self:UIKit Tab Group Object Reference=@n9, bSuppressClickFeedback:bool=true) -> (then)
   "On Pressed (TouchArea_Left)"() {is=title K2Node_ComponentBoundEvent}
   > FireEventElementInteractedByName(Element Name=OnPressed_TouchArea_Left, WidgetCallingFrom=Self-Reference())
   > SelectPreviousTab!(execute, self:UIKit Tab Group Object Reference=@n9, bSuppressClickFeedback:bool=true) -> (then)
   n10 = "On Clicked (Apply_Button)"!() -> (OutputDelegate, then, Button:Common Button Base Object Reference) {is=title K2Node_ComponentBoundEvent}
   > Fire Event Element Interacted(Element=@n10.Button)
   > n11 = ConfirmAndApplyPreset()
   > n12 = Branch(Condition=@n11)
   CloseScreen(execute=[@n13.then, @n12.then])
   n13 = FortShippingLog(
    execute=@n12.else
    WorldContextObject=GetOwningPlayer()
    InString="WBP_Pilgrim_ControlSchemes::OnClicked (Apply_Button) -- Failed to save preset."
    bLogAsWarning=false
   )
   n14 = "On Clicked (Back_Button)"!() -> (OutputDelegate, then, Button:Common Button Base Object Reference) {is=title K2Node_ComponentBoundEvent}
   > Fire Event Element Interacted(Element=@n14.Button)
   > CloseScreen()
   SetOverdriveText() {is=title K2Node_CustomEvent}
   > n15 = PlaylistDataReadyAsync!(execute:Exec) -> (
    then:Exec
    OnReady:Exec
    GameState:Fort Game State Athena Object Reference
    Playlist:Fort Playlist Object Reference
    PlaylistContextTags:"Gameplay Tag Container Structure (by ref)"
   ) {is=title K2Node_AsyncAction}
   n16 = Cast To FortPlaylistAthena!(execute=@n15.OnReady, Object=@n15.Playlist) -> (then, CastFailed, AsFort Playlist Athena:Fort Playlist Athena Object Reference) {is=title K2Node_DynamicCast}
   n17 = MakeLiteralGameplayTagContainer!(Value:Gameplay Tag Container Structure=(GameplayTags=((TagName="Sparks.Playlist.Pilgrim.BattleStage")))) -> (ReturnValue:Gameplay Tag Container Structure)
   n18 = Get GameplayTagContainer!(self:Fort Playlist Object Reference=@n16) -> (GameplayTagContainer:Gameplay Tag Container Structure)
   > "SetText (Text)"(
    self=Get Overdrive_Text() -> (Overdrive_Text:Common Text Object Reference)
    InText=Select(
     Option 0:Text="LOCTEXT(\"OVERDRIVE\")"
     Option 1:Text="LOCTEXT(\"ATTACK\")"
     Index=HasAnyTags!(
      TagContainer:"Gameplay Tag Container Structure (by ref)"=@n18
      OtherContainer:"Gameplay Tag Container Structure (by ref)"=@n17
      bExactMatch:bool=false
     ) -> (ReturnValue:bool)
    ) -> (ReturnValue:Text)
   )
   Event OnInitialized()
   > Parent: OnInitialized() {is=title K2Node_CallParentFunction}
   > n19 = PilgrimWaitForEventRouter!(execute:Exec, ContextActor:Actor Object Reference=GetOwningPlayer()) -> (then:Exec, OnEventRouterAvailable:Exec, EventRouter:Gameplay Event Router Component Object Reference) {is=title K2Node_AsyncAction}
   n20 = StartListeningToEvent!(
    execute:Exec=@n19.OnEventRouterAvailable
    Target:Gameplay Event Router Component Object Reference=@n19
    EventType:Script Struct Object Reference=/Script/PilgrimQPSharedRuntime.PilgrimEvent_QuickplayState_Began
    Context:Object Reference
    EventBubblingRule:EEventBubblingRule Enum
   ) -> (
    then:Exec
    AsyncTaskProxy:Async Action Start Listening To Event Object Reference
    OnEventReceived:Exec
    Payload:Pilgrim Event Quickplay State Began Structure
   )
   Branch(
    execute=@n20.OnEventReceived
    Condition="Not Equal (Enum)"!(
     A:EPilgrimQuickplayState Enum=Break Pilgrim Event Quickplay State Base!(PilgrimEvent_QuickplayState_Base:"Pilgrim Event Quickplay State Base Structure (by ref)"=@n20.Payload) -> (StateId:EPilgrimQuickplayState Enum) {is=title K2Node_BreakStruct}
     B:EPilgrimQuickplayState Enum=Invalid
    ) -> (ReturnValue:bool)
   )
   > CloseScreen()
   Event Destruct()
   > Unbind all Events from On Input Method Changed!(
    execute:Exec
    self:Common Input Object Reference=Get CommonInputSubsystem!(PlayerController:Player Controller Object Reference=GetOwningPlayer()) -> (ReturnValue:Common Input Object Reference) {is=title K2Node_GetSubsystemFromPC}
   ) -> (then:Exec) {is=title K2Node_ClearDelegate}
   n21 = Event PreConstruct()
   > Branch(Condition=@n21.IsDesignTime)
   > Set TabIndex(TabIndex=Get TabIndex() -> (TabIndex=0)) -> (Output_Get=0)
   > SetupTabs()
   > SetControlsVisuals()
   Event Construct()
   > SetupPlayerIdentity!(
    execute
    self:WBP Pilgrim Player Identity Pill Object Reference=Get WBP_Pilgrim_PlayerIdentity_Pill() -> (WBP_Pilgrim_PlayerIdentity_Pill:WBP Pilgrim Player Identity Pill Object Reference)
    OptionalPlayerState:Player State Object Reference
    ForceVisible:bool=false
   ) -> (then)
   > SetupTabs()
   > SetCurrentMappingContext()
   > SetOverdriveText() {is=title K2Node_CallFunction}
   > SetControlsVisuals()
   > SetInputMappings()
   > SetPresetName()
   > SetStageTechFTUEVisuals()
   n22 = Get CommonInputSubsystem(PlayerController=GetOwningPlayer())
   > Bind Event to On Input Method Changed(self=@n22, Delegate=Create Event())
   > OnInputMethodChanged!(execute, self, InputMethod:ECommonInputType Enum=GetCurrentInputType(self=@n22)) -> (then)
   > SetAllNavigationRules!(execute, self:Widget Object Reference, Rule:EUINavigationRule Enum=Stop, WidgetToFocus:Name=None) -> (then)
   Event On Activated()
   > Fire Event Main View Opened!(execute, View:Widget Object Reference=Self-Reference(), AdditionalInfo:Flow Analytics Additional Info Structure, bIsOverlay:bool=false) -> (then)
   > SetFocus(self=Get WBP_Pilgrim_ControlSchemesSelector())
  ]}
 GetAppliedPreset {nodes [
   GetAppliedPreset() {is=title K2Node_FunctionEntry}
   n1 = GetOwningPlayer()
   > n2 = Switch on ECommonInputType!(execute, Selection:ECommonInputType Enum=GetCurrentInputType(self=Get CommonInputSubsystem(PlayerController=@n1))) -> (MouseAndKeyboard:Exec, Gamepad:Exec, Touch:Exec, Count:Exec) {is=title K2Node_SwitchEnum}
   n3 = GetPilgrimClientSettingRecordPartition!(self:Pilgrim Player Controller Component Object Reference=GetComponentByClass!(
     self:Actor Object Reference=GetOwningPlayer()
     ComponentClass:Actor Component Class Reference=/Script/PilgrimCoreRuntime.PilgrimPlayerControllerComponent
    ) -> (ReturnValue:Pilgrim Player Controller Component Object Reference)) -> (ReturnValue:Pilgrim Client Setting Record Partition Object Reference)
   n4 = GetKeyboardAndMouseExpertPresetIndex!(self:Pilgrim Client Setting Record Partition Object Reference=@n3) -> (ReturnValue:int=0)
   n5 = GetKeyboardAndMouseGeneralPresetIndex!(self:Pilgrim Client Setting Record Partition Object Reference=@n3) -> (ReturnValue:int=0)
   > Set ControlsIndex(ControlsIndex=Select(
     Option 0:int=Select!(Option 0=0, Option 1=0, Option 2:int=1, Option 3:int=0, Option 4:int=2, Index=@n5) -> (ReturnValue)
     Option 1:int=Select(Option 0=0, Option 1=0, Option 2=1, Option 3=0, Option 4=2, Index=@n4)
     Index:int=Get TabIndex() -> (TabIndex=0)
    ) -> (ReturnValue:int)) -> (Output_Get=0)
   n6 = GetPilgrimClientSettingRecordPartition(self=GetComponentByClass(self=GetOwningPlayer(), ComponentClass=&PilgrimPlayerControllerComponent))
   n7 = GetControllerGeneralPresetIndex!(self:Pilgrim Client Setting Record Partition Object Reference=@n6) -> (ReturnValue:int=0)
   n8 = GetControllerExpertPresetIndex!(self:Pilgrim Client Setting Record Partition Object Reference=@n6) -> (ReturnValue:int=0)
   Set ControlsIndex(
    execute=@n2.Gamepad
    ControlsIndex=Select!(
     Option 0=Select!(Option 0=0, Option 1=0, Option 2:int=1, Option 3:int=2, Option 4:int=0, Index=@n7) -> (ReturnValue)
     Option 1=Select(Option 0=0, Option 1=0, Option 2=1, Option 3=2, Option 4=0, Index=@n8)
     Index=Get TabIndex() -> (TabIndex=0)
    ) -> (ReturnValue)
   ) -> (Output_Get=0)
  ]}
 GetDeviceType {nodes [
   GetDeviceType() {is=title K2Node_FunctionEntry}
   n1 = GetOwningPlayer()
   > n2 = Switch on ECommonInputType(Selection=GetCurrentInputType(self=Get CommonInputSubsystem(PlayerController=@n1)))
   Return Node!(execute=@n2.Count, Device Type:EHardwareDevicePrimaryType Enum=Unspecified)
   Return Node(execute=@n2.Gamepad, Device Type=Gamepad)
   > Return Node(Device Type=KeyboardAndMouse)
   Return Node(execute=@n2.Touch, Device Type=Touch)
  ]}
 GetInputImages {nodes [
   n1 = GetInputImages!() -> (then, KeybindIndex:int) {is=title K2Node_FunctionEntry}
   n2 = Get TabIndex() -> (TabIndex=0)
   n3 = Get OverdriveInput_Image() -> (OverdriveInput_Image:Image Object Reference)
   n4 = Get OverdriveInput_Image()
   n5 = Get TabIndex() -> (TabIndex=0)
   n6 = Get Key_Image5() -> (Key_Image5:Image Object Reference)
   > n7 = Set KeybindingImages(KeybindingImages:Array of Image Object References=Make Array!(
     "[0]":Image Object Reference=Get Key_Image1() -> (Key_Image1:Image Object Reference)
     "[1]":Image Object Reference=Get Key_Image2() -> (Key_Image2:Image Object Reference)
     "[2]":Image Object Reference=Get Key_Image3() -> (Key_Image3:Image Object Reference)
     "[3]":Image Object Reference=Get Key_Image4() -> (Key_Image4:Image Object Reference)
     "[4]":Image Object Reference=Select!(Option 0:Image Object Reference=@n4, Option 1:Image Object Reference=@n6, Index=@n5) -> (ReturnValue:Image Object Reference)
     "[5]":Image Object Reference=Select(Option 1=@n3, Index=@n2)
    ) -> (Array:Array of Image Object References)) -> (Output_Get:Array of Image Object References)
   > n8 = Branch(Condition=Is Valid Index!(TargetArray:Array of Image Object References=@n7, IndexToTest:int=@n1) -> (ReturnValue:bool))
   > Return Node!(
    execute
    KeybindImage:Image Object Reference="Get (a copy)"(
     Array:Array of Image Object References=Get KeybindingImages() -> (KeybindingImages:Array of Image Object References)
     Dimension 1=Get KeybindIndex() -> (KeybindIndex:int=0)
    ) -> (Output:Image Object Reference)
   )
   Return Node(execute=@n8.else)
  ]}
 OnInputMethodChanged {nodes [
   OnInputMethodChanged!() -> (then, InputMethod:ECommonInputType Enum) {is=title K2Node_FunctionEntry}
   > GetAppliedPreset() {is=title K2Node_CallFunction}
   > SetCurrentMappingContext()
   > SetInputMappings()
   > SetPresetName()
  ]}
 OnKeyDown {nodes [
   n1 = OnKeyDown!() -> (then, MyGeometry:Geometry Structure, InKeyEvent:Key Event Structure) {is=title K2Node_FunctionEntry}
   > Branch(Condition=Get IsFTUE())
   n2 = GetKey!(Input:"Key Event Structure (by ref)"=@n1.InKeyEvent) -> (ReturnValue:Key Structure)
   > n3 = Branch(Condition="Equal (Key)"(A:Key Structure=@n2, B:Key Structure=Gamepad_LeftShoulder) -> (ReturnValue:bool=~) {is=title K2Node_PromotableOperator})
   > SelectPreviousTab(self=Get WBP_UIKit_Tabs_TopNavBar(), bSuppressClickFeedback=true)
   > Return Node!(execute, ReturnValue:Event Reply Structure=Handled())
   n4 = Branch(execute=@n3.else, Condition="Equal (Key)"(A=@n2, B=Gamepad_RightShoulder) -> (ReturnValue=~))
   > SelectNextTab(self=Get WBP_UIKit_Tabs_TopNavBar(), bSuppressClickFeedback=true)
   > Return Node(ReturnValue=Handled())
   Return Node(execute=@n4.else, ReturnValue=Unhandled())
  ]}
 OpenSettings {nodes [
   OpenSettings() {is=title K2Node_FunctionEntry}
   n1 = GetFortPlayerControllerFromActor!(Actor:Actor Object Reference=GetOwningPlayer()) -> (ReturnValue:Fort Player Controller Object Reference)
   > n2 = Create Widget!(
    execute:Exec
    Class:User Widget Class Reference=/FortniteCoreUI/Settings/SettingScreen.SettingScreen_C
    OwningPlayer:Player Controller Object Reference=@n1
   ) -> (then:Exec, ReturnValue:Setting Screen Object Reference)
   > PushContentWidgetForPlayer!(execute, LocalPlayerController:Fort Player Controller Object Reference=@n1, PushWidget:Widget Object Reference=@n2) -> (then)
   n3 = IsUsingGamepad!(self:Fort Player Controller Object Reference=@n1) -> (ReturnValue:bool)
   > NavigateToSetting!(
    execute
    self:Fort Settings Screen Object Reference=@n2
    SettingCategory:Name=Select(
     Option 0:Name=Get SettingsCategoryMKB() -> (SettingsCategoryMKB:Name=None)
     Option 1:Name=Get SettingsCategoryGamepad() -> (SettingsCategoryGamepad:Name=None)
     Index:bool=@n3
    ) -> (ReturnValue:Name)
    SettingToSelect:String=Select(
     Option 0:String=Get SettingToSelectMKB() -> (SettingToSelectMKB:String)
     Option 1:String=Get SettingToSelectGamepad() -> (SettingToSelectGamepad:String)
     Index=@n3
    ) -> (ReturnValue:String)
   ) -> (then)
  ]}
 SendApplyControlsPresetAnalytics {nodes [
   n1 = SendApplyControlsPresetAnalytics!() -> (then, Num Lanes:int, Device Type:EHardwareDevicePrimaryType Enum=Unspecified, Selected Preset Index:int) {is=title K2Node_FunctionEntry}
   > n2 = ConstructAnalyticsAttribute!(
    execute
    AttributeName:String=Selected_Num_Lanes
    StringValue:String
    FGuidValue:Guid Structure
    IntValue:int=@n1
    floatValue:float=-1
    BoolValue:bool=false
   ) -> (then, ReturnValue:Fort Analytics Event Attribute Structure)
   > Add(
    TargetArray:Array of Fort Analytics Event Attribute Structures=Get AttributeArray() -> (AttributeArray:Array of Fort Analytics Event Attribute Structures)
    NewItem:"Fort Analytics Event Attribute Structure (by ref)"=@n2
   ) {is=title K2Node_CallArrayFunction}
   > n3 = ConstructAnalyticsAttribute(
    AttributeName=Device_Type
    StringValue=Enum to String!(Enumerator:EHardwareDevicePrimaryType Enum=@n1.Device Type) -> (ReturnValue:String)
    IntValue=-1
    floatValue=-1
    BoolValue=false
   )
   > Add(TargetArray=Get AttributeArray(), NewItem=@n3)
   > n4 = ConstructAnalyticsAttribute(AttributeName=Selected_Preset_Index, IntValue=@n1.Selected Preset Index, floatValue=-1, BoolValue=false)
   > Add(TargetArray=Get AttributeArray(), NewItem=@n4)
   > FireAnalyticsEvent_BlueprintArray!(
    execute
    EventName:String=Pilgrim.Player.ApplyControlsPreset
    Attributes:Array of Fort Analytics Event Attribute Structures=Get AttributeArray()
   ) -> (then)
  ]}
 SetControlsVisuals {nodes [
   SetControlsVisuals() {is=title K2Node_FunctionEntry}
   n1 = Get TabIndex() -> (TabIndex=0)
   > SetBrushFromTexture!(
    execute
    self:Image Object Reference=Get ControlsTexture_Image() -> (ControlsTexture_Image:Image Object Reference)
    Texture:Texture 2D Object Reference=Select(
     Option 0:Texture 2D Object Reference=/PilgrimCore/Widgets/Textures/T_UI_Pilgrim_4-Lane
     Option 1:Texture 2D Object Reference=/PilgrimCore/Widgets/Textures/T_UI_Pilgrim_5-Lane
     Index:int=@n1
    ) -> (ReturnValue:Texture 2D Object Reference)
    bMatchSize:bool=false
   ) -> (then)
   > SetVisibility(
    self=Get Key_Image5()
    InVisibility=Select(Option 0:ESlateVisibility Enum=Collapsed, Option 1:ESlateVisibility Enum=SelfHitTestInvisible, Index=@n1) -> (ReturnValue:ESlateVisibility Enum)
   )
   > SetSize!(
    execute
    self:Spacer Object Reference=Get Binds_Spacer() -> (Binds_Spacer:Spacer Object Reference)
    InSize:Vector 2D Structure=Select(Option 0:Vector 2D Structure=(X=550,Y=0), Option 1:Vector 2D Structure=(X=700,Y=0), Index=@n1) -> (ReturnValue:Vector 2D Structure)
   ) -> (then)
  ]}
 SetCurrentMappingContext {nodes [
   SetCurrentMappingContext() {is=title K2Node_FunctionEntry}
   n1 = Get TabIndex() -> (TabIndex=0)
   > Set CurrentMappingContexts(CurrentMappingContexts:Array of Fort Input Mapping Context Object References=GetDeviceInputMappingContexts!(
     self:Pilgrim Preset Configuration Object Reference=Get PilgrimPresetConfiguration() -> (PilgrimPresetConfiguration:Pilgrim Preset Configuration Object Reference)
     NumLanes:int=Select(Option 0:int=4, Option 1:int=5, Index=@n1) -> (ReturnValue:int)
     DeviceType:EHardwareDevicePrimaryType Enum=GetDeviceType() -> (Device Type=Unspecified) {is=title K2Node_CallFunction}
    ) -> (ReturnValue:Array of Fort Input Mapping Context Object References)) -> (Output_Get:Array of Fort Input Mapping Context Object References)
  ]}
 SetInputMappings {nodes [
   SetInputMappings() {is=title K2Node_FunctionEntry}
   > n1 = Branch(Condition="Equal (Integer)"(A=Get ControlsIndex() -> (ControlsIndex=0), B=2))
   > ShowAppliedInputs()
   ShowPresets(execute=@n1.else)
  ]}
 SetIsFTUE {nodes [
   SetIsFTUE() {is=title K2Node_FunctionEntry}
   n1 = GetComponentByClass(self=GetOwningPlayer(), ComponentClass=/PilgrimCore/Player/Components/BP_Pilgrim_FTUEComponent.BP_Pilgrim_FTUEComponent_C) -> (ReturnValue:BP Pilgrim FTUEComponent Object Reference)
   > n2 = Is Valid(InputObject=@n1)
   Set IsFTUE(execute=@n2.Is Not Valid, IsFTUE:bool=false) -> (Output_Get:bool)
   n3 = HasUserSeenFTUE!(execute=@n2.Is Valid, self:BP Pilgrim FTUEComponent Object Reference=@n1) -> (then, HasSeenFTUE:bool)
   > Set IsFTUE(IsFTUE=NOT Boolean(A=@n3))
  ]}
 SetPresetName {nodes [
   SetPresetName() {is=title K2Node_FunctionEntry}
   n1 = GetOwningPlayer()
   n2 = "Equal (Enum)"(A:ECommonInputType Enum=GetCurrentInputType(self=Get CommonInputSubsystem(PlayerController=@n1)), B:ECommonInputType Enum=MouseAndKeyboard)
   n3 = Get ControlsIndex() -> (ControlsIndex=0)
   n4 = Select!(Option 0:Text="LOCTEXT(\"Two Hands\")", Option 1:Text="LOCTEXT(\"One Hand\")", Option 2:Text="LOCTEXT(\"Custom\")", Index=@n3) -> (ReturnValue:Text)
   > Set 4LanesText(4LanesText:Text=SelectText(
     A=@n4
     B=Select(Option 0="LOCTEXT(\"Thumbs\")", Option 1="LOCTEXT(\"Index && Middle\")", Option 2="LOCTEXT(\"Custom\")", Index=@n3)
     bPickA=@n2
    )) -> (Output_Get:Text)
   > Set 5LanesText(5LanesText:Text=SelectText(
     A=@n4
     B=Select(Option 0="LOCTEXT(\"Thumbs\")", Option 1="LOCTEXT(\"Thumb, Index && Middle\")", Option 2="LOCTEXT(\"Custom\")", Index=@n3)
     bPickA=@n2
    )) -> (Output_Get:Text)
   > "SetText (Text)"(
    self=Get SliderValue!(self:WBP Pilgrim Control Schemes Selector Object Reference=Get WBP_Pilgrim_ControlSchemesSelector()) -> (SliderValue:Common Text Object Reference)
    InText=Select!(Option 0=Get 4LanesText() -> (4LanesText:Text), Option 1=Get 5LanesText() -> (5LanesText:Text), Index=Get TabIndex() -> (TabIndex=0)) -> (ReturnValue)
   )
  ]}
 SetStageTechFTUEVisuals {nodes [
   SetStageTechFTUEVisuals() {is=title K2Node_FunctionEntry}
   > SetIsFTUE() {is=title K2Node_CallFunction}
   n1 = Get IsFTUE()
   > SetActiveWidgetIndex(
    self=Get NextButtonSwitcher() -> (NextButtonSwitcher:Common Activatable Widget Switcher Object Reference)
    Index="To Integer (Boolean)"!(InBool:bool=@n1) -> (ReturnValue:int=0)
   )
   > SetVisibility(
    self=Get BottomButtons() -> (BottomButtons:Grid Panel Object Reference)
    InVisibility=Select(Option 0:ESlateVisibility Enum=SelfHitTestInvisible, Option 1:ESlateVisibility Enum=Collapsed, Index:bool=@n1) -> (ReturnValue:ESlateVisibility Enum)
   )
   > SetIsEnabled(
    self=[
     Get Apply_Button() -> (Apply_Button:WBP UIKit Button Quiet Object Reference)
     Get Back_Button() -> (Back_Button:WBP UIKit Button Quiet Object Reference)
    ]
    bInIsEnabled=NOT Boolean(A=@n1)
   )
  ]}
 SetupTabs {nodes [
   SetupTabs() {is=title K2Node_FunctionEntry}
   > Clear(TargetArray:Array of UI Kit Tab Group Button View Model Object References=Get Local Tabs View Models() -> (Local Tabs View Models:Array of UI Kit Tab Group Button View Model Object References))
   n1 = Get WBP_UIKit_Tabs_TopNavBar()
   > Set Tabs Data!(
    execute
    self:WBP UIKit Tabs Base Object Reference=@n1
    InTabs Data:Array of UI Kit Tab Group Button View Model Object References=Get Local Tabs View Models()
   ) -> (then) {is=title K2Node_CallFunction}
   > n2 = For Each Loop(Array:Array of Texts=Get Tabs() -> (Tabs:Array of Texts)) -> (Array Element:Text)
   > n3 = Construct Object from Class(Class=/Script/FortniteUI.UIKitTabGroupButtonViewModel) -> (ReturnValue:UI Kit Tab Group Button View Model Object Reference)
   > Set Label!(execute, Label:Text=@n2, self:UI Kit Selectable Button View Model Object Reference=@n3) -> (then, Output_Get:Text)
   > Add(
    TargetArray:Array of UI Kit Tab Group Button View Model Object References=Get Local Tabs View Models()
    NewItem:"UI Kit Tab Group Button View Model Object Reference (by ref)"=@n3
   )
   Set Tabs Data(execute=@n2.Completed, self=@n1, InTabs Data=Get Local Tabs View Models())
   > SelectTabByIndex!(execute, self:UIKit Tab Group Object Reference=Get WBP_UIKit_Tabs_TopNavBar(), Index:int=0, bSuppressClickFeedback:bool=true) -> (then, ReturnValue:bool)
   > GetAppliedPreset()
   > n4 = Branch(Condition=Get IsFTUE())
   > SetInputActionsType!(
    execute
    self:WBP UIKit Tabs Base Object Reference=Get WBP_UIKit_Tabs_TopNavBar()
    InputActions:E_UI_TabsInputActionType Enum=NewEnumerator0
    InputActionsWithMouseScroll:bool=false
   ) -> (then)
   SetInputActionsType(execute=@n4.else, self=Get WBP_UIKit_Tabs_TopNavBar(), InputActions=NewEnumerator3, InputActionsWithMouseScroll=false)
  ]}
 ShowAppliedInputs {nodes [
   ShowAppliedInputs() {is=title K2Node_FunctionEntry}
   > Clear(TargetArray:Array of Names=Get MappingNames() -> (MappingNames:Array of Names))
   n1 = Select(
    Option 0:Input Mapping Context Object Reference=Get 4LanesDefault() -> (4LanesDefault:Fort Input Mapping Context Object Reference)
    Option 1:Input Mapping Context Object Reference=Get 5LanesDefault() -> (5LanesDefault:Fort Input Mapping Context Object Reference)
    Index:int=Get TabIndex() -> (TabIndex=0)
   ) -> (ReturnValue:Input Mapping Context Object Reference)
   > n2 = For Each Loop(Array:Array of Enhanced Action Key Mapping Structures=Break Input Mapping Context Mapping Data!(InputMappingContextMappingData:"Input Mapping Context Mapping Data Structure (by ref)"=Get DefaultKeyMappings!(self:Input Mapping Context Object Reference=@n1) -> (DefaultKeyMappings:Input Mapping Context Mapping Data Structure)) -> (Mappings:Array of Enhanced Action Key Mapping Structures) {is=title K2Node_BreakStruct}) -> (Array Element:Enhanced Action Key Mapping Structure)
   > Add Unique(
    TargetArray:Array of Names=Get MappingNames()
    NewItem:"Name (by ref)"=GetMappingName!(ActionKeyMapping:"Enhanced Action Key Mapping Structure (by ref)"=@n2) -> (ReturnValue:Name=None)
   )
   n3 = For Each Loop(
    Exec=@n2.Completed
    Array:Array of Slate Brush Structures=GetKeyBrushesPlayerCustomMappings!(
     self:Pilgrim Preset Previewer Object Reference
     PlayerController:Player Controller Object Reference=GetOwningPlayer()
     MappingNames:Array of Names=Get MappingNames()
     DeviceType:EHardwareDevicePrimaryType Enum=GetDeviceType() -> (Device Type=Unspecified)
    ) -> (ReturnValue:Array of Slate Brush Structures)
   ) -> (Array Element:Slate Brush Structure)
   > n4 = GetInputImages!(execute, self, KeybindIndex:int=@n3.Array Index) -> (then, KeybindImage:Image Object Reference) {is=title K2Node_CallFunction}
   > Is Valid(InputObject=@n4)
   > SetBrush(self=@n4, InBrush=@n3)
  ]}
 ShowPresets {nodes [
   ShowPresets() {is=title K2Node_FunctionEntry}
   n1 = Select(Option 0:int=0, Option 1:int=2, Index:bool="Equal (Integer)"(A=Get ControlsIndex() -> (ControlsIndex=0), B=1)) -> (ReturnValue:int)
   n2 = Get CurrentMappingContexts() -> (CurrentMappingContexts:Array of Fort Input Mapping Context Object References)
   > n3 = For Each Loop(Array=GetKeyBrushesForMappingContext!(
     self:Pilgrim Preset Previewer Object Reference
     InputMappingContext:Fort Input Mapping Context Object Reference="Get (a copy)"(Array:Array of Fort Input Mapping Context Object References=@n2, Dimension 1=@n1) -> (Output:Fort Input Mapping Context Object Reference)
     DeviceType:EHardwareDevicePrimaryType Enum=GetDeviceType() -> (Device Type=Unspecified)
    ) -> (ReturnValue:Array of Slate Brush Structures))
   > n4 = GetInputImages(KeybindIndex=@n3.Array Index)
   > Is Valid(InputObject=@n4)
   > SetBrush(self=@n4, InBrush=@n3)
  ]}
]
variables [
 4LanesDefault {type=FortInputMappingContext, default=/PilgrimSettings/Input/SharedAssets/IMC_Pilgrim_4LaneDefault}
 4LanesText {type=string, default=""}
 5LanesDefault {type=FortInputMappingContext, default=/PilgrimSettings/Input/SharedAssets/IMC_Pilgrim_5LaneDefault}
 5LanesText {type=string, default=""}
 ControlsIndex {type=integer}
 CurrentMappingContexts {type="array<FortInputMappingContext>", default []}
 IsFTUE {type=boolean}
 Local Tabs View Models {type="array<UIKitTabGroupButtonViewModel>", default []}
 MappingNames {type="array<string>", default []}
 PilgrimPresetConfiguration {type=PilgrimPresetConfiguration, default=/PilgrimSettings/Settings/PilgrimPresetConfiguration}
 PresetList {type="array<string>", default [Thumb, "Index && Middle", "Thumb, Index && Middle", Two Hands, One Hand]}
 SettingsCategoryGamepad {type=string, default=ControllerBindingsCollection}
 SettingsCategoryMKB {type=string, default=MouseAndKeyboardBindingsCollection}
 SettingToSelectGamepad {type=string, default=Pilgrim_Preset_4Lanes_Gamepad}
 SettingToSelectMKB {type=string, default=Pilgrim_Preset_4Lanes_KeyboardAndMouse}
 TabIndex {type=integer, default=1}
 Tabs {type="array<string>", default ["Easy, Medium, Hard", Expert]}
]
widgets {
 info {inheritedWidgetCount=0, parentClass=/Script/PilgrimSettingsUI.PilgrimPresetPreviewer, rootWidgetClass=&GridPanel, widgetCount=43}
 widgets [
  GridPanel_21 {widgetClassPath=&GridPanel, widget=GridPanel_21, properties {columnFill [1], rowFill [1]}}
  WBP_FMCalibration_SharedBackground {
   widgetClassPath=/FMCalibration/Widgets/WBP_FMCalibration_SharedBackground
   bIsVariable=true
   parent=GridPanel_21
   widget=WBP_FMCalibration_SharedBackground
   slot {class=GridSlot}
  }
  WBP_CaptureForPostBufferUpdate {
   widgetClassPath=&WBP_CaptureForPostBufferUpdate
   bIsVariable=true
   parent=GridPanel_21
   slot {class=GridSlot, properties {layer=1}}
  }
  Border_1 {
   widgetClassPath=&Border
   parent=GridPanel_21
   widget=Border_1
   properties {
    background {
     bIsDynamicallyLoaded=false
     drawAs=Image
     imageType=FullColor
     mirroring=NoMirror
     resourceName=None
     resourceObject=/FMCore/Materials/MusicLibraryBG_Inst_DublicateFMCore
     tiling=NoTile
     imageSize=<32 32>
     margin=<ltrb 0 0 0 0>
     outlineSettings {
      bUseBrushTransparency=false
      roundingType=HalfHeightRadius
      width=0
      color {colorUseRule=UseColor_Specified, specifiedColor=<rgba 0 0 0 0>}
      cornerRadii=<0 0 0 0>
     }
     tintColor {colorUseRule=UseColor_Specified, specifiedColor=<rgba 1 1 1 1>}
     uVRegion {bIsValid=false, max=<0 0>, min=<0 0>}
    }
    padding=<ltrb 0 40 0 40>
   }
   slot {class=GridSlot, properties {layer=2, padding=<ltrb 0 100 0 100>}}
  }
  GridPanel_2 {
   widgetClassPath=&GridPanel
   parent=Border_1
   widget=GridPanel_2
   properties {columnFill [1], rowFill [1]}
   slot {class=BorderSlot, properties {padding=<ltrb 0 40 0 40>}}
  }
  VerticalBox_2 {
   widgetClassPath=&VerticalBox
   parent=GridPanel_2
   widget=VerticalBox_2
   slot {class=GridSlot, properties {horizontalAlignment=HAlign_Center, verticalAlignment=VAlign_Top}}
  }
  Title_Text {
   widgetClassPath=&CommonTextBlock
   bIsVariable=true
   parent=VerticalBox_2
   widget=Title_Text
   properties {
    clipping=OnDemand
    flowDirectionPreference=Culture
    text=GAMEPLAY CONTROLS
    textTransformPolicy=ToUpper
    font {
     bForceMonospaced=false
     bMaterialIsStencil=false
     fontMaterial=null
     fontObject=&HeadingNow
     letterSpacing=0
     monospacedWidth=1
     size=24
     skewAmount=0
     typefaceFontName=76 Bold Italic
     outlineSettings {
      bApplyOutlineToDropShadows=false
      bMiteredCorners=false
      bSeparateFillAlpha=false
      outlineMaterial=&M-FontShadow
      outlineSize=2
      outlineColor=<rgba 0 0 0 1>
     }
    }
    shadowOffset=<0 0>
    strikeBrush {
     bIsDynamicallyLoaded=false
     drawAs=NoDrawType
     imageType=NoImage
     mirroring=NoMirror
     resourceName=None
     resourceObject=null
     tiling=NoTile
     imageSize=<32 32>
     margin=<ltrb 0 0 0 0>
     outlineSettings {
      bUseBrushTransparency=false
      roundingType=HalfHeightRadius
      width=0
      color {colorUseRule=UseColor_Specified, specifiedColor=<rgba 0 0 0 0>}
      cornerRadii=<0 0 0 0>
     }
     tintColor {colorUseRule=UseColor_Specified, specifiedColor=<rgba 1 1 1 1>}
     uVRegion {bIsValid=false, max=<0 0>, min=<0 0>}
    }
   }
   slot {class=VerticalBoxSlot, properties {horizontalAlignment=HAlign_Center}}
  }
  SubTittle_Text {
   widgetClassPath=&CommonTextBlock
   bIsVariable=true
   parent=VerticalBox_2
   widget=SubTittle_Text
   properties {
    clipping=OnDemand
    flowDirectionPreference=Culture
    text=Set controls for when difficulty is
    textTransformPolicy=ToUpper
    colorAndOpacity {colorUseRule=UseColor_Specified, specifiedColor=<rgba 1 1 1 0.7>}
    font {
     bForceMonospaced=false
     bMaterialIsStencil=false
     fontMaterial=null
     fontObject=&HeadingNow
     letterSpacing=0
     monospacedWidth=1
     size=24
     skewAmount=0
     typefaceFontName=76 Bold Italic
     outlineSettings {
      bApplyOutlineToDropShadows=false
      bMiteredCorners=false
      bSeparateFillAlpha=false
      outlineMaterial=&M-FontShadow
      outlineSize=2
      outlineColor=<rgba 0 0 0 1>
     }
    }
    shadowOffset=<0 0>
    strikeBrush {
     bIsDynamicallyLoaded=false
     drawAs=NoDrawType
     imageType=NoImage
     mirroring=NoMirror
     resourceName=None
     resourceObject=null
     tiling=NoTile
     imageSize=<32 32>
     margin=<ltrb 0 0 0 0>
     outlineSettings {
      bUseBrushTransparency=false
      roundingType=HalfHeightRadius
      width=0
      color {colorUseRule=UseColor_Specified, specifiedColor=<rgba 0 0 0 0>}
      cornerRadii=<0 0 0 0>
     }
     tintColor {colorUseRule=UseColor_Specified, specifiedColor=<rgba 1 1 1 1>}
     uVRegion {bIsValid=false, max=<0 0>, min=<0 0>}
    }
   }
   slot {class=VerticalBoxSlot}
  }
  SZ_TopTabs {
   widgetClassPath=&SafeZone
   parent=GridPanel_2
   widget=SZ_TopTabs
   properties {padBottom=false, padLeft=false, padRight=false}
   slot {class=GridSlot, properties {horizontalAlignment=HAlign_Center, verticalAlignment=VAlign_Top, padding=<ltrb 0 60 0 0>}}
  }
  TopTabs_Overlay {widgetClassPath=&Overlay, parent=SZ_TopTabs, widget=TopTabs_Overlay, slot {class=SafeZoneSlot}}
  VerticalBox_3 {
   widgetClassPath=&VerticalBox
   parent=TopTabs_Overlay
   widget=VerticalBox_3
   slot {class=OverlaySlot, properties {horizontalAlignment=HAlign_Fill, verticalAlignment=VAlign_Fill}}
  }
  HorizontalBox_0 {widgetClassPath=&HorizontalBox, parent=VerticalBox_3, widget=HorizontalBox_0, slot {class=VerticalBoxSlot}}
  TouchArea_Left {
   widgetClassPath=&Button
   parent=HorizontalBox_0
   widget=TouchArea_Left
   properties {
    isFocusable=false
    backgroundColor=<rgba 1 1 1 0>
    colorAndOpacity=<rgba 1 1 1 0>
    widgetStyle {
     clickedSlateSound {resourceObject=null}
     disabled {
      bIsDynamicallyLoaded=false
      drawAs=NoDrawType
      imageType=NoImage
      mirroring=NoMirror
      resourceName=None
      resourceObject=null
      tiling=NoTile
      imageSize=<0 0>
      margin=<ltrb 0 0 0 0>
      outlineSettings {
       bUseBrushTransparency=false
       roundingType=HalfHeightRadius
       width=0
       color {colorUseRule=UseColor_Specified, specifiedColor=<rgba 0 0 0 0>}
       cornerRadii=<0 0 0 0>
      }
      tintColor {colorUseRule=UseColor_Specified, specifiedColor=<rgba 1 1 1 0>}
      uVRegion {bIsValid=false, max=<0 0>, min=<0 0>}
     }
     disabledForeground {colorUseRule=UseColor_Specified, specifiedColor=<rgba 0.5271 0.5271 0.5271 0>}
     hovered {
      bIsDynamicallyLoaded=false
      drawAs=RoundedBox
      imageType=NoImage
      mirroring=NoMirror
      resourceName=None
      resourceObject=null
      tiling=NoTile
      imageSize=<32 32>
      margin=<ltrb 0 0 0 0>
      outlineSettings {
       bUseBrushTransparency=true
       roundingType=FixedRadius
       width=1
       color {colorUseRule=UseColor_Specified, specifiedColor=<rgba 0.7243 0.7243 0.7243 1>}
       cornerRadii=<4 4 4 4>
      }
      tintColor {colorUseRule=UseColor_Specified, specifiedColor=<rgba 0.7243 0.7243 0.7243 0>}
      uVRegion {bIsValid=false, max=<0 0>, min=<0 0>}
     }
     hoveredForeground {colorUseRule=UseColor_Specified, specifiedColor=<rgba 0 1 1 0>}
     hoveredSlateSound {resourceObject=null}
     normal {
      bIsDynamicallyLoaded=false
      drawAs=RoundedBox
      imageType=NoImage
      mirroring=NoMirror
      resourceName=None
      resourceObject=null
      tiling=NoTile
      imageSize=<32 32>
      margin=<ltrb 0 0 0 0>
      outlineSettings {
       bUseBrushTransparency=true
       roundingType=FixedRadius
       width=1
       color {colorUseRule=UseColor_Specified, specifiedColor=<rgba 0.6951 0.6951 0.6951 1>}
       cornerRadii=<4 4 4 4>
      }
      tintColor {colorUseRule=UseColor_Specified, specifiedColor=<rgba 0.4955 0.4955 0.4955 0>}
      uVRegion {bIsValid=false, max=<0 0>, min=<0 0>}
     }
     normalForeground {colorUseRule=UseColor_Specified, specifiedColor=<rgba 0 1 1 0>}
     normalPadding=<ltrb 12 1.5 12 1.5>
     pressed {
      bIsDynamicallyLoaded=false
      drawAs=RoundedBox
      imageType=NoImage
      mirroring=NoMirror
      resourceName=None
      resourceObject=null
      tiling=NoTile
      imageSize=<32 32>
      margin=<ltrb 0 0 0 0>
      outlineSettings {
       bUseBrushTransparency=true
       roundingType=FixedRadius
       width=1
       color {colorUseRule=UseColor_Specified, specifiedColor=<rgba 0.7243 0.7243 0.7243 1>}
       cornerRadii=<4 4 4 4>
      }
      tintColor {colorUseRule=UseColor_Specified, specifiedColor=<rgba 0.3843 0.3843 0.3843 0>}
      uVRegion {bIsValid=false, max=<0 0>, min=<0 0>}
     }
     pressedForeground {colorUseRule=UseColor_Specified, specifiedColor=<rgba 0 1 1 0>}
     pressedPadding=<ltrb 12 1.5 12 1.5>
     pressedSlateSound {resourceObject=null}
    }
   }
   slot {class=HorizontalBoxSlot, properties {size {sizeRule=Fill, value=1}}}
  }
  TouchArea_Right {
   widgetClassPath=&Button
   parent=HorizontalBox_0
   widget=TouchArea_Right
   properties {
    isFocusable=false
    backgroundColor=<rgba 1 1 1 0>
    colorAndOpacity=<rgba 1 1 1 0>
    widgetStyle {
     clickedSlateSound {resourceObject=null}
     disabled {
      bIsDynamicallyLoaded=false
      drawAs=NoDrawType
      imageType=NoImage
      mirroring=NoMirror
      resourceName=None
      resourceObject=null
      tiling=NoTile
      imageSize=<0 0>
      margin=<ltrb 0 0 0 0>
      outlineSettings {
       bUseBrushTransparency=false
       roundingType=HalfHeightRadius
       width=0
       color {colorUseRule=UseColor_Specified, specifiedColor=<rgba 0 0 0 0>}
       cornerRadii=<0 0 0 0>
      }
      tintColor {colorUseRule=UseColor_Specified, specifiedColor=<rgba 1 1 1 0>}
      uVRegion {bIsValid=false, max=<0 0>, min=<0 0>}
     }
     disabledForeground {colorUseRule=UseColor_Specified, specifiedColor=<rgba 0.5271 0.5271 0.5271 0>}
     hovered {
      bIsDynamicallyLoaded=false
      drawAs=RoundedBox
      imageType=NoImage
      mirroring=NoMirror
      resourceName=None
      resourceObject=null
      tiling=NoTile
      imageSize=<32 32>
      margin=<ltrb 0 0 0 0>
      outlineSettings {
       bUseBrushTransparency=true
       roundingType=FixedRadius
       width=1
       color {colorUseRule=UseColor_Specified, specifiedColor=<rgba 0.7243 0.7243 0.7243 1>}
       cornerRadii=<4 4 4 4>
      }
      tintColor {colorUseRule=UseColor_Specified, specifiedColor=<rgba 0.7243 0.7243 0.7243 0>}
      uVRegion {bIsValid=false, max=<0 0>, min=<0 0>}
     }
     hoveredForeground {colorUseRule=UseColor_Specified, specifiedColor=<rgba 0 1 1 0>}
     hoveredSlateSound {resourceObject=null}
     normal {
      bIsDynamicallyLoaded=false
      drawAs=RoundedBox
      imageType=NoImage
      mirroring=NoMirror
      resourceName=None
      resourceObject=null
      tiling=NoTile
      imageSize=<32 32>
      margin=<ltrb 0 0 0 0>
      outlineSettings {
       bUseBrushTransparency=true
       roundingType=FixedRadius
       width=1
       color {colorUseRule=UseColor_Specified, specifiedColor=<rgba 0.6951 0.6951 0.6951 1>}
       cornerRadii=<4 4 4 4>
      }
      tintColor {colorUseRule=UseColor_Specified, specifiedColor=<rgba 0.4955 0.4955 0.4955 0>}
      uVRegion {bIsValid=false, max=<0 0>, min=<0 0>}
     }
     normalForeground {colorUseRule=UseColor_Specified, specifiedColor=<rgba 0 1 1 0>}
     normalPadding=<ltrb 12 1.5 12 1.5>
     pressed {
      bIsDynamicallyLoaded=false
      drawAs=RoundedBox
      imageType=NoImage
      mirroring=NoMirror
      resourceName=None
      resourceObject=null
      tiling=NoTile
      imageSize=<32 32>
      margin=<ltrb 0 0 0 0>
      outlineSettings {
       bUseBrushTransparency=true
       roundingType=FixedRadius
       width=1
       color {colorUseRule=UseColor_Specified, specifiedColor=<rgba 0.7243 0.7243 0.7243 1>}
       cornerRadii=<4 4 4 4>
      }
      tintColor {colorUseRule=UseColor_Specified, specifiedColor=<rgba 0.3843 0.3843 0.3843 0>}
      uVRegion {bIsValid=false, max=<0 0>, min=<0 0>}
     }
     pressedForeground {colorUseRule=UseColor_Specified, specifiedColor=<rgba 0 1 1 0>}
     pressedPadding=<ltrb 12 1.5 12 1.5>
     pressedSlateSound {resourceObject=null}
    }
   }
   slot {class=HorizontalBoxSlot, properties {size {sizeRule=Fill, value=1}}}
  }
  Overlay_0 {widgetClassPath=&Overlay, parent=VerticalBox_3, widget=Overlay_0, slot {class=VerticalBoxSlot}}
  TabBacking {
   widgetClassPath=&Image
   parent=Overlay_0
   widget=TabBacking
   properties {
    visibility=SelfHitTestInvisible
    brush {
     bIsDynamicallyLoaded=false
     drawAs=Image
     imageType=NoImage
     mirroring=NoMirror
     resourceName=None
     resourceObject=/PilgrimCore/Widgets/Materials/MI_Pilgrim_TabBacking_Panel
     tiling=NoTile
     imageSize=<32 32>
     margin=<ltrb 0 0 0 0>
     outlineSettings {
      bUseBrushTransparency=false
      roundingType=HalfHeightRadius
      width=0
      color {colorUseRule=UseColor_Specified, specifiedColor=<rgba 0 0 0 0>}
      cornerRadii=<0 0 0 0>
     }
     tintColor {colorUseRule=UseColor_Specified, specifiedColor=<rgba 1 1 1 1>}
     uVRegion {bIsValid=false, max=<0 0>, min=<0 0>}
    }
   }
   slot {class=OverlaySlot, properties {horizontalAlignment=HAlign_Fill, verticalAlignment=VAlign_Fill, padding=<ltrb -10 0 -10 0>}}
  }
  WBP_UIKit_Tabs_TopNavBar {
   widgetClassPath=/Game/UI/Frontend/TopBar/WBP_UIKit_Tabs_TopNavBar
   bIsVariable=true
   parent=Overlay_0
   widget=WBP_UIKit_Tabs_TopNavBar
   properties {
    bAutoListenForInput=true
    backplateClass=null
    buttonClass=/PilgrimCore/Widgets/SparksUILibrary/WBP_Sparks_TextTabs_Button_Selectable.WBP_Sparks_TextTabs_Button_Selectable_C
    buttonVerticalAlignment=VAlign_Fill
    hideMouseAndKeyboardPrompts=true
    initialSelectedTabId=1
    initialSelectedTabIndex=0
    inputActions=Bumpers
    promptBackplateCornerRadius=Interactive Component
    promptBackplateMinWidth=64
    reservedPromptSpace=32
    selectionAlignment=Left
    tabGroupVerticalAlignment=VAlign_Center
    tabSpacing=0
    useBackplate=true
    iconSize=<0 0>
    localTabsViewModels [null,null]
    tabGroupPadding=<ltrb 0 8 0 8>
   }
   slot {class=OverlaySlot, properties {padding=<ltrb 10 0 10 0>}}
  }
  GridPanel_0 {
   widgetClassPath=&GridPanel
   parent=GridPanel_2
   widget=GridPanel_0
   slot {class=GridSlot, properties {horizontalAlignment=HAlign_Center, verticalAlignment=VAlign_Center, padding=<ltrb 0 0 0 20>}}
  }
  Overlay_1 {widgetClassPath=&Overlay, parent=GridPanel_0, widget=Overlay_1, slot {class=GridSlot}}
  ControlsTexture_Image {
   widgetClassPath=&Image
   bIsVariable=true
   parent=Overlay_1
   widget=ControlsTexture_Image
   properties {
    visibility=SelfHitTestInvisible
    brush {
     bIsDynamicallyLoaded=false
     drawAs=Image
     imageType=NoImage
     mirroring=NoMirror
     resourceName=None
     resourceObject=&T_UI_Pilgrim_5-Lane
     tiling=NoTile
     imageSize=<1024 340>
     margin=<ltrb 0 0 0 0>
     outlineSettings {
      bUseBrushTransparency=false
      roundingType=HalfHeightRadius
      width=0
      color {colorUseRule=UseColor_Specified, specifiedColor=<rgba 0 0 0 0>}
      cornerRadii=<0 0 0 0>
     }
     tintColor {colorUseRule=UseColor_Specified, specifiedColor=<rgba 1 1 1 1>}
     uVRegion {bIsValid=false, max=<0 0>, min=<0 0>}
    }
   }
   slot {class=OverlaySlot}
  }
  Keybinds_GridPanel {
   widgetClassPath=&GridPanel
   parent=GridPanel_0
   widget=Keybinds_GridPanel
   slot {class=GridSlot, properties {horizontalAlignment=HAlign_Center, verticalAlignment=VAlign_Bottom}}
  }
  Keybinds_HorizontalBox {widgetClassPath=&HorizontalBox, bIsVariable=true, parent=Keybinds_GridPanel, widget=Keybinds_HorizontalBox, slot {class=GridSlot}}
  Key_Image1 {
   widgetClassPath=&Image
   bIsVariable=true
   parent=Keybinds_HorizontalBox
   widget=Key_Image1
   properties {visibility=SelfHitTestInvisible}
   slot {class=HorizontalBoxSlot, properties {horizontalAlignment=HAlign_Center, size {sizeRule=Fill, value=1}}}
  }
  Key_Image2 {
   widgetClassPath=&Image
   bIsVariable=true
   parent=Keybinds_HorizontalBox
   widget=Key_Image2
   properties {visibility=SelfHitTestInvisible}
   slot {class=HorizontalBoxSlot, properties {horizontalAlignment=HAlign_Center, size {sizeRule=Fill, value=1}}}
  }
  Key_Image3 {
   widgetClassPath=&Image
   bIsVariable=true
   parent=Keybinds_HorizontalBox
   widget=Key_Image3
   properties {visibility=SelfHitTestInvisible}
   slot {class=HorizontalBoxSlot, properties {horizontalAlignment=HAlign_Center, size {sizeRule=Fill, value=1}}}
  }
  Key_Image4 {
   widgetClassPath=&Image
   bIsVariable=true
   parent=Keybinds_HorizontalBox
   widget=Key_Image4
   properties {visibility=SelfHitTestInvisible}
   slot {class=HorizontalBoxSlot, properties {horizontalAlignment=HAlign_Center, size {sizeRule=Fill, value=1}}}
  }
  Key_Image5 {
   widgetClassPath=&Image
   bIsVariable=true
   parent=Keybinds_HorizontalBox
   widget=Key_Image5
   properties {visibility=SelfHitTestInvisible}
   slot {class=HorizontalBoxSlot, properties {horizontalAlignment=HAlign_Center, size {sizeRule=Fill, value=1}}}
  }
  Binds_Spacer {
   widgetClassPath=&Spacer
   bIsVariable=true
   parent=Keybinds_GridPanel
   widget=Binds_Spacer
   properties {size=<700 0>}
   slot {class=GridSlot}
  }
  Overdrive_GridPanel {
   widgetClassPath=&GridPanel
   parent=GridPanel_0
   widget=Overdrive_GridPanel
   properties {columnFill [1], renderTransform {angle=0, scale=<1 1>, shear=<0 0>, translation=<880 0>}, rowFill [1]}
   slot {class=GridSlot, properties {horizontalAlignment=HAlign_Left, verticalAlignment=VAlign_Center}}
  }
  Border_0 {
   widgetClassPath=&Border
   parent=Overdrive_GridPanel
   widget=Border_0
   properties {
    visibility=SelfHitTestInvisible
    background {
     bIsDynamicallyLoaded=false
     drawAs=Image
     imageType=FullColor
     mirroring=NoMirror
     resourceName=None
     resourceObject=/PilgrimCore/Widgets/Materials/MI_UI_Overdrive_ProgressMeter_Glow_Activated
     tiling=NoTile
     imageSize=<32 32>
     margin=<ltrb 0 0 0 0>
     outlineSettings {
      bUseBrushTransparency=false
      roundingType=HalfHeightRadius
      width=0
      color {colorUseRule=UseColor_Specified, specifiedColor=<rgba 0 0 0 0>}
      cornerRadii=<0 0 0 0>
     }
     tintColor {colorUseRule=UseColor_Specified, specifiedColor=<rgba 1 1 1 1>}
     uVRegion {bIsValid=false, max=<0 0>, min=<0 0>}
    }
   }
   slot {class=GridSlot, properties {horizontalAlignment=HAlign_Center, verticalAlignment=VAlign_Top}}
  }
  Overdrive_Text {
   widgetClassPath=&CommonTextBlock
   bIsVariable=true
   parent=Border_0
   widget=Overdrive_Text
   properties {
    justification=Center
    scrollStyle=&TextScrollStyle-Base_C
    text=OVERDRIVE
    font {
     bForceMonospaced=false
     bMaterialIsStencil=false
     fontMaterial=/FortUILibrary/UIKit/Buttons/MaterialInstances/GenericButtons/MI_UIKit_GenericText_Quiet
     fontObject=&HeadingNow
     letterSpacing=0
     monospacedWidth=1
     size=16.5
     skewAmount=0
     typefaceFontName=86 Bold
     outlineSettings {
      bApplyOutlineToDropShadows=false
      bMiteredCorners=false
      bSeparateFillAlpha=false
      outlineMaterial=null
      outlineSize=0
      outlineColor=<rgba 0 0 0 1>
     }
    }
    shadowOffset=<0 0>
   }
   slot {class=BorderSlot}
  }
  OverdriveInput_Image {
   widgetClassPath=&Image
   bIsVariable=true
   parent=Overdrive_GridPanel
   widget=OverdriveInput_Image
   properties {visibility=SelfHitTestInvisible}
   slot {class=GridSlot, properties {horizontalAlignment=HAlign_Center, verticalAlignment=VAlign_Bottom, padding=<ltrb 0 30 0 0>}}
  }
  CoachMark_GridPanel {
   widgetClassPath=&SizeBox
   parent=GridPanel_0
   widget=CoachMark_GridPanel
   properties {bOverride_HeightOverride=true, heightOverride=160.4802}
   slot {class=GridSlot, properties {horizontalAlignment=HAlign_Right, verticalAlignment=VAlign_Center}}
  }
  WBP_UIKit_CoachMark_1 {
   widgetClassPath=/FortUILibrary/UIKit/CoachMark/WBP_UIKit_CoachMark
   bIsVariable=true
   parent=CoachMark_GridPanel
   widget=WBP_UIKit_CoachMark_1
   properties {
    arrowPosition=Right_Middle
    bSupportsActivationFocus=false
    dismissButtonClass=null
    justificationDescription=Center
    otherActionButtonClass=null
    safeZone=40
    shouldShowDismissButton=false
    textDescription="When you're on stage, your gameplay controls will look like this."
    textDismissButton=""
    textHeader=""
    textOtherActionButton=""
    visibility=SelfHitTestInvisible
    widget Width=430
    brush_Icon {
     bIsDynamicallyLoaded=false
     drawAs=Image
     imageType=NoImage
     mirroring=NoMirror
     resourceName=None
     resourceObject=null
     tiling=NoTile
     imageSize=<0 0>
     margin=<ltrb 0 0 0 0>
     outlineSettings {
      bUseBrushTransparency=false
      roundingType=HalfHeightRadius
      width=0
      color {colorUseRule=UseColor_Specified, specifiedColor=<rgba 0 0 0 0>}
      cornerRadii=<0 0 0 1>
     }
     tintColor {colorUseRule=UseColor_Specified, specifiedColor=<rgba 1 1 1 1>}
     uVRegion {bIsValid=false, max=<0 0>, min=<0 0>}
    }
    in Decorator Classes []
    renderTransform {angle=0, scale=<1 1>, shear=<0 0>, translation=<-929 0>}
   }
   slot {class=SizeBoxSlot}
  }
  ButtonStack_VerticalBox {
   widgetClassPath=&VerticalBox
   parent=GridPanel_2
   widget=ButtonStack_VerticalBox
   properties {navigation=&self:WidgetTree.ButtonStack_VerticalBox.WidgetNavigation_0}
   slot {class=GridSlot, properties {horizontalAlignment=HAlign_Center, verticalAlignment=VAlign_Bottom}}
  }
  WBP_Pilgrim_ControlSchemesSelector {
   widgetClassPath=&/WBP_Pilgrim_ControlSchemesSelector
   bIsVariable=true
   parent=ButtonStack_VerticalBox
   widget=WBP_Pilgrim_ControlSchemesSelector
   properties {bIsFocusable=true}
   slot {class=VerticalBoxSlot}
  }
  NextButtonSwitcher {
   widgetClassPath=&CommonActivatableWidgetSwitcher
   bIsVariable=true
   parent=ButtonStack_VerticalBox
   widget=NextButtonSwitcher
   properties {transitionDuration=0}
   slot {class=VerticalBoxSlot, properties {padding=<ltrb 0 10 0 0>}}
  }
  CustomControl_Button {
   widgetClassPath=/SparksUICommon/UILibrary/WBP_FM_Core_Button_Regular
   bIsVariable=true
   parent=NextButtonSwitcher
   widget=CustomControl_Button
   properties {
    bDisplayInputActionWhenNotInteractable=false
    bHideInputActionWithKeyboard=true
    bUseNaturalNavigation=true
    buttonSize=Small
    icon=/Game/UI/Sidebar/Textures/T_UI_Icon_Sidebar_Settings
    text=Set Custom Controls
   }
   slot {class=WidgetSwitcherSlot}
  }
  Next_FTUE_Button {
   widgetClassPath=&/WBP_PilgrimCoreFTUE_NextButton
   bIsVariable=true
   parent=NextButtonSwitcher
   widget=Next_FTUE_Button
   properties {text="Apply && Next"}
   slot {class=WidgetSwitcherSlot, properties {padding=<ltrb 0 20 0 0>}}
  }
  BottomButtons {
   widgetClassPath=&GridPanel
   bIsVariable=true
   parent=GridPanel_21
   widget=BottomButtons
   slot {class=GridSlot, properties {horizontalAlignment=HAlign_Right, layer=2, verticalAlignment=VAlign_Bottom}}
  }
  Apply_Button {
   widgetClassPath=/FortUILibrary/UIKit/Buttons/WBP_UIKit_Button_Quiet
   bIsVariable=true
   parent=BottomButtons
   widget=Apply_Button
   properties {text=Apply, triggeringInputAction {dataTable=&InputActionDataTable, rowName=GenericFaceButtonLeft}}
   slot {class=GridSlot, properties {column=2, padding=<ltrb 20 20 20 20>}}
  }
  Back_Button {
   widgetClassPath=&WBP_UIKit_Button_Quiet
   bIsVariable=true
   parent=BottomButtons
   widget=Back_Button
   properties {text=Back, triggeringInputAction {dataTable=&InputActionDataTable, rowName=Back}}
   slot {class=GridSlot, properties {column=3, padding=<ltrb 20 20 20 20>}}
  }
  WBP_Pilgrim_PlayerIdentity_Pill {
   widgetClassPath=/PilgrimCore/Widgets/Player/WBP_Pilgrim_PlayerIdentity_Pill
   bIsVariable=true
   parent=GridPanel_21
   widget=WBP_Pilgrim_PlayerIdentity_Pill
   properties {padding=<ltrb 0 23 23 0>}
   slot {class=GridSlot, properties {horizontalAlignment=HAlign_Right, layer=2, verticalAlignment=VAlign_Top}}
  }
 ]
}
